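# Parent Niagara system for motorcycle wheel, tire and exhaust effects, including surface-based dust and debris, rear-wheel peel-outs, mud ribbon trails and gear-shift flames.
# User inputs such as speed, wheelie, stoppie, peel-out and deep water combine with system surface flags and timelines to gate and scale spawning.
# Emitters pack variants into spawn groups, masking look and motion per group and tagging renderer visibility, with skeletal mesh sockets positioning particles.
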NiagaraSystem /Motorcycle/FX/Parent/NS_Motorcycle_Parent_Wheel_And_Tire_Interaction
dependencies {
 usedDataInterfaces [
  &NiagaraDataInterfaceCurve
  &NiagaraDataInterfaceArrayFloat
  &NiagaraDataInterfaceArrayInt32
  &NiagaraDataInterfaceArrayFloat3
  &NiagaraDataInterfaceSpriteRendererInfo
  &NiagaraDataInterfaceVectorCurve
  &NiagaraDataInterfaceMeshRendererInfo
  &NiagaraDataInterfaceColorCurve
  &NiagaraDataInterfaceSkeletalMesh
  &NiagaraDataInterfaceArrayColor
 ]
 usedDynamicInputs [&SimulationPosition, &FN_LOD_Distance_Remap]
 usedModules [
  &self:SystemSpawnScript.SystemScriptSource.SystemScriptGraph.NiagaraNodeAssignment_1.SetVariables_#1
  &self:SystemSpawnScript.SystemScriptSource.SystemScriptGraph.NiagaraNodeAssignment_2.SetVariables_#2
  &self:SystemSpawnScript.SystemScriptSource.SystemScriptGraph.NiagaraNodeAssignment_3.SetVariables_#3
  &SystemState
  &self:SystemSpawnScript.SystemScriptSource.SystemScriptGraph.NiagaraNodeAssignment_0.SetVariables_#4
  &self:SetPeeloutFXArrayIndex
  &self:SparksSpawn
  &Timeline
 ]
}
emitters {
 Dust_Debris {
  emitterUpdateScript [
   &EmitterState()
   &SpawnRate(
    SpawnRate=&Multiply_Float(
     A=&Multiply_Float(
      A=&RandomRangeFloat(Minimum=14, Maximum=56)
      B=&MakeCustomFloatFromBool(
       Bool=&Boolean_AndOperation(
        Bool1=$User.Dust_Wheel_Rear
        Bool2=&Boolean_AndOperation(
         Bool1=&Boolean_OrOperation(
          Bool1=$System.SpawnRoadSmoke
          Bool2=&Boolean_OrOperation(Bool1=$System.SpawnDirtClods, Bool2=&Boolean_OrOperation(Bool1=$System.SpawnSnow, Bool2=$System.SpawnSand))
         )
         Bool2=&Boolean_NotOperation(Bool=&Boolean_OrOperation(Bool1=$User.PeelOut_Wheel_Rear, Bool2=$User.IsWaterDeep))
        )
       )
       True Float=&MakeCustomFloatFromBool(Bool=&Boolean_OrOperation(Bool1=$User.IsOnWheelie, Bool2=$User.IsDoingStoppie), True Float=0.5, False Float=1)
       False Float=0
      )
     )
     B=&FloatFromCurve(FloatCurve=$System.Dust_SpawnCurve, CurveIndex=&Abs_Float(Float:NiagaraFloat=$User.Vehicle_Speed))
    )
    Use Spawn Probability=false
   )
   &SpawnRate(
    SpawnGroup=1
    SpawnRate=&Multiply_Float(
     A=&Multiply_Float(
      A=&RandomRangeFloat(Minimum=7, Maximum=28)
      B=&MakeCustomFloatFromBool(
       Bool=&Boolean_AndOperation(Bool1=$User.Dust_Wheel_Rear, Bool2=&Boolean_OrOperation(Bool1=$System.SpawnDirtClods, Bool2=$System.SpawnGrass))
       True Float=&MakeCustomFloatFromBool(Bool=&Boolean_OrOperation(Bool1=$User.IsOnWheelie, Bool2=$User.IsDoingStoppie), True Float=0.5, False Float=1)
       False Float=0
      )
     )
     B=&FloatFromCurve(FloatCurve=$System.Dust_SpawnCurve, CurveIndex=&Abs_Float(Float:NiagaraFloat=$User.Vehicle_Speed))
    )
    Use Spawn Probability=false
   ) {module=SpawnRate001}
   &SpawnRate(
    SpawnGroup=2
    SpawnRate=&Multiply_Float(
     A=&Multiply_Float(
      A=&RandomRangeFloat(Minimum=42, Maximum=165)
      B=&MakeCustomFloatFromBool(
       Bool=&Boolean_AndOperation(Bool1=$User.Dust_Wheel_Rear, Bool2=$System.SpawnDirtClods)
       True Float=&MakeCustomFloatFromBool(Bool=&Boolean_OrOperation(Bool1=$User.IsOnWheelie, Bool2=$User.IsDoingStoppie), True Float=0.5, False Float=1)
       False Float=0
      )
     )
     B=&FloatFromCurve(FloatCurve=$System.Dust_SpawnCurve, CurveIndex=&Abs_Float(Float:NiagaraFloat=$User.Vehicle_Speed))
    )
    Use Spawn Probability=false
   ) {module=SpawnRate002}
   &SpawnRate(
    SpawnGroup=3
    SpawnRate=&MakeCustomFloatFromBool(
     Bool=&Boolean_AndOperation(
      Bool1=$User.Dust_Wheel_Rear
      Bool2=&Boolean_AndOperation(
       Bool1=$System.SpawnMud
       Bool2=&Boolean_NotOperation(Bool=&Boolean_OrOperation(Bool1=$User.PeelOut_Wheel_Rear, Bool2=&Boolean_OrOperation(Bool1=$User.IsWaterDeep, Bool2=$User.IsDoingStoppie)))
      )
     )
     True Float=&Multiply_Float(
      A=65
      B=&Lerp_Float(A=0, B=1, Alpha=&ClampFloat(Float=&Power(a=$User.Vehicle_Speed, b=0.4)))
     )
     False Float=0
    )
    Use Spawn Probability=false
   ) {module=SpawnRate003}
   &SpawnRate(
    SpawnGroup=4
    SpawnRate=&Multiply_Float(
     A=6
     B=&MakeCustomFloatFromBool(
      Bool=&Boolean_AndOperation(Bool1=$User.PeelOut_Wheel_Rear, Bool2=&Boolean_NotOperation(Bool=$User.TirePoppedSparksRear))
      True Float=1
      False Float=0
     )
    )
    Use Spawn Probability=false
   ) {module=SpawnRate004}
  ]
  particleSpawnScript [
   /SetVariables(Particles.VisibilityTag=&SpawnGroup()) {isSetParameters=true}
   &PartitionParticles()
   &InitializeParticle(
    Color=&MaskLinearColorBySpawnGroup(
     Masked Linear Color=&MultiplyLinearColor(Linear Color=$System.PeelOutColor, Scale Factor=<1 1 1 0.2>)
     Passthrough Linear Color=&MaskLinearColorBySpawnGroup(
      Masked Linear Color=<rgba 0.375 0.375 0.375 1>
      Passthrough Linear Color=&MaskLinearColorBySpawnGroup(
       Masked Linear Color=&MultiplyLinearColor(Linear Color=$System.PeelOutColor, Scale Factor=<1 1 1 1.2>)
       Passthrough Linear Color=$System.PeelOutColor
       Spawn Group Mask=2
      )
      Spawn Group Mask=1
     )
    )
    Lifetime=&MaskFloatBySpawnGroup(
     Masked Float=&Lerp_Float(
      A=&RandomRangeFloat(Minimum=0.2, Maximum=0.22)
      B=1
      Alpha=&Abs_Float(Float:NiagaraFloat=$User.Vehicle_Speed)
     )
     Passthrough Float=&MaskFloatBySpawnGroup(
      Masked Float=&RandomRangeFloat(Minimum=0.5, Maximum=1)
      Passthrough Float=&MaskFloatBySpawnGroup(
       Masked Float=&Lerp_Float(
        A=&RandomRangeFloat(Minimum=1.5, Maximum=2.5)
        B=1
        Alpha=&Abs_Float(Float:NiagaraFloat=$User.Vehicle_Speed)
       )
       Passthrough Float=&MaskFloatBySpawnGroup(
        Masked Float=&Lerp_Float(
         A=&RandomRangeFloat(Minimum=0.6, Maximum=0.8)
         B=1.3
         Alpha=&Abs_Float(Float:NiagaraFloat=$User.Vehicle_Speed)
        )
        Passthrough Float=0.5
        Spawn Group Mask=3
       )
       Spawn Group Mask=2
      )
      Spawn Group Mask=1
     )
    )
    Mesh Scale=<0.9 0.5 0.9>
    Mesh Scale Mode=::NewEnumerator4
    Sprite Rotation Mode=::NewEnumerator2
    Sprite Size Mode=::NewEnumerator1
    Uniform Sprite Size Max=&MaskFloatBySpawnGroup(
     Masked Float=6
     Passthrough Float=&MaskFloatBySpawnGroup(
      Masked Float=&Lerp_Float(A=36, B=48, Alpha=&Abs_Float(Float:NiagaraFloat=$User.Vehicle_Speed))
      Passthrough Float=&Lerp_Float(A=60, B=80, Alpha=&Abs_Float(Float:NiagaraFloat=$User.Vehicle_Speed))
      Spawn Group Mask=3
     )
     Spawn Group Mask=1
    )
    Uniform Sprite Size Min=&MaskFloatBySpawnGroup(
     Masked Float=2
     Passthrough Float=&MaskFloatBySpawnGroup(
      Masked Float=&Lerp_Float(A=24, B=39, Alpha=&Abs_Float(Float:NiagaraFloat=$User.Vehicle_Speed))
      Passthrough Float=&Lerp_Float(A=40, B=65, Alpha=&Abs_Float(Float:NiagaraFloat=$User.Vehicle_Speed))
      Spawn Group Mask=3
     )
     Spawn Group Mask=1
    )
    UsePositionOffset=false
    Write Material Random=true
   )
   &SubUVAnimation(
    Evaluation Type=::NewEnumerator0
    "Sprite Renderer "=DataInterface() {propertyValues {SpriteRenderer=&self:Dust_Debris_0.NiagaraSpriteRendererProperties_0}}
    SubUV Animation Mode:ENiagara_SubUVLookupModeV2=::NewEnumerator2
    UseEndFrame=false
    UseStartFrame=false
   )
   &SkeletalMeshLocation(
    Bone / Socket Index=&MaskIntBySpawnGroup(
     Masked Int=1
     Passthrough Int=&MaskIntBySpawnGroup(
      Masked Int=1
      Passthrough Int=&MakeCustomIntFromBool(
       Bool=$User.IsOnWheelie
       True INT=1
       False INT=&MakeCustomIntFromBool(Bool=$User.IsDoingStoppie, True INT=0, False INT=$Particles.PartitionParticles.Partition)
      )
      Spawn Group Mask=3
     )
     Spawn Group Mask=2
    )
    MaskSpawnGroup=false
    Mesh Sampling Type=::NewEnumerator5
    Skeletal Mesh=$System.SkeletalMesh
    Socket Sampling Mode=::NewEnumerator5
   )
   &ShapeLocation(MaskSpawnGroup=true, Override Seed=false, Spawn Group Mask=1, Sphere Radius=5, UseOwnerScale=true)
   &AddVelocity(
    Cone Angle=35
    Cone Axis=<1 0 1>
    Module.UseConeFalloff=false
    Module.UseVelDistribution=false
    Override Seed=false
    Velocity Mode=::NewEnumerator2
    Velocity Speed=&MaskFloatBySpawnGroup(Masked Float=&RandomRangeFloat(Minimum=50, Maximum=500), Passthrough Float=0, Spawn Group Mask=1)
   )
   &AddVelocity(
    Velocity=&MaskVectorBySpawnGroup(
     Masked Vector=&RandomRangeVector(Minimum=<500 -50 50>, Maximum=<600 50 1200>)
     Passthrough Vector=&MaskVectorBySpawnGroup(
      Masked Vector=&RandomRangeVector(
       Minimum=<500 -150 0>
       Maximum=&MakeVector(
        X=600
        Y=150
        Z=&Lerp_Float(
         A=&MakeCustomFloatFromBool(Bool=$User.IsPeelOutOnRoad, True Float=0, False Float=500)
         B=1400
         Alpha=$User.Vehicle_Speed
        )
       )
      )
      Passthrough Vector=<0 0 0>
      Spawn Group Mask=3
     )
     Spawn Group Mask=2
    )
    Velocity Speed Scale=&Lerp_Float(A=0.2, B=1, Alpha=&Abs_Float(Float:NiagaraFloat=$User.Vehicle_Speed))
   ) {module=AddVelocity001}
   &InitialMeshOrientation(
    Mesh Orientation Mode=::NewEnumerator7
    Rotation=&RandomRangeVector(Minimum=<0 0 0>, Maximum=<0 1 0>)
    Use Rotation Vector=true
   )
  ]
  particleUpdateScript [
   &ParticleState()
   &ScaleSpriteSize(
    Scale Sprite Size Mode=::NewEnumerator0
    Uniform Scale Factor=&MaskFloatBySpawnGroup(
     Masked Float=&FloatFromCurve(
      FloatCurve=DataInterface() {propertyValues {Curve {keys [{interpMode=RCIM_Cubic, tangentMode=RCTM_Break, time=0, value=0.179}, <at 0.8329 1>]}}}
      Scale Curve=3
     )
     Passthrough Float=&MaskFloatBySpawnGroup(
      Masked Float=&FloatFromCurve(FloatCurve=DataInterface() {propertyValues {Curve {keys [
           {tangentMode=RCTM_User, time=0, value=1}
           {arriveTangent=-0.0405, interpMode=RCIM_Cubic, leaveTangent=-0.0405, tangentMode=RCTM_User, time=0.7009, value=1}
           {arriveTangent=-8.0089, interpMode=RCIM_Cubic, leaveTangent=-8.0089, tangentMode=RCTM_User, time=1, value=0}
          ]}}})
      Passthrough Float=&MaskFloatBySpawnGroup(
       Masked Float=&FloatFromCurve(
        FloatCurve=DataInterface() {propertyValues {Curve {keys [
            {interpMode=RCIM_Cubic, leaveTangent=0.9665, tangentMode=RCTM_Break, time=0, value=0.1}
            {arriveTangent=1.0806, interpMode=RCIM_Cubic, leaveTangent=1.0806, time=0.3541, value=0.5745}
            {arriveTangent=1.9934, interpMode=RCIM_Cubic, leaveTangent=1.9934, tangentMode=RCTM_User, time=0.8329, value=1}
           ]}}}
        Scale Curve=5
       )
       Passthrough Float=&FloatFromCurve(
        FloatCurve=DataInterface() {propertyValues {Curve {keys [
            {interpMode=RCIM_Cubic, leaveTangent=1.2011, tangentMode=RCTM_Break, time=0, value=0.179}
            {arriveTangent=0.6691, interpMode=RCIM_Cubic, leaveTangent=0.6691, tangentMode=RCTM_User, time=0.8329, value=1}
           ]}}}
        Scale Curve=&Multiply_Float(
         A=&Lerp_Float(A=3.5, B=5.5, Alpha=$System.Timeline002.Time)
         B=&MakeCustomFloatFromBool(Bool=$User.IsPeelOutOnRoad, True Float=0.6, False Float=1)
        )
       )
       Spawn Group Mask=2
      )
      Spawn Group Mask=1
     )
    )
   )
   &SpriteRotationRate(Rotation Rate=&MaskFloatBySpawnGroup(
     Masked Float=0
     Passthrough Float=&Multiply_Float(A=$Particles.Initial.MaterialRandom, B=1)
     Spawn Group Mask=1
    ))
   &GravityForce(Gravity=&MaskVectorBySpawnGroup(
     Masked Vector=<0 0 0>
     Passthrough Vector=&MaskVectorBySpawnGroup(
      Masked Vector=<0 0 -2940>
      Passthrough Vector=&MaskVectorBySpawnGroup(Masked Vector=<0 0 0>, Passthrough Vector=<0 0 -980>, Spawn Group Mask=4)
      Spawn Group Mask=3
     )
    ))
   &Drag(
    Drag=&MaskFloatBySpawnGroup(
     Masked Float=0
     Passthrough Float=&MaskFloatBySpawnGroup(Masked Float=3, Passthrough Float=1, Spawn Group Mask=3)
     Spawn Group Mask=1
    )
    Rotational Drag:NiagaraFloat=&MaskFloatBySpawnGroup(Masked Float=0, Passthrough Float=1, Spawn Group Mask=1)
    Use Linear Drag=true
    Use Rotational Drag=true
   )
   &InheritVelocity(
    Inherited Velocity Amount Scale=&MaskVectorBySpawnGroup(
     Masked Vector=&VectorFromCurve(
      VectorCurve=DataInterface() {propertyValues {
        XCurve {keys [{arriveTangent=2.817, interpMode=RCIM_Cubic, leaveTangent=2.817, tangentMode=RCTM_User, time=0, value=0}, <at 1 1 RCIM_Cubic>]}
        YCurve {keys [{arriveTangent=2.817, interpMode=RCIM_Cubic, leaveTangent=2.817, tangentMode=RCTM_User, time=0, value=0}, <at 1 1 RCIM_Cubic>]}
        ZCurve {keys [{arriveTangent=2.817, interpMode=RCIM_Cubic, leaveTangent=2.817, tangentMode=RCTM_User, time=0, value=0}, <at 1 1 RCIM_Cubic>]}
       }}
      Scale Curve=<0.8 0.8 0.8>
     )
     Passthrough Vector=<0 0 0>
    )
    Speed Limit=true
    UseSpeedThreshold=false
   )
   &InheritSourceMovement(
    Applied Position Scale=&MaskVectorBySpawnGroup(
     Masked Vector=<0.9 0.9 0.9>
     Passthrough Vector=&MaskVectorBySpawnGroup(Masked Vector=<0.9 0.9 0.9>, Passthrough Vector=<0 0 0>, Spawn Group Mask=3)
     Spawn Group Mask=2
    )
    ApplyForce=false
    ApplyPosition=true
    ApplyPositionFalloff=false
   )
   &SolveForcesAndVelocity(Clamp Velocity=false, Limit Acceleration=false)
   &DynamicMaterialParameters(
    Index 0 Param 1=&Add_Float(
     A=&RandomRangeFloat(Minimum=-1, Maximum=1)
     B=&FloatFromCurve(FloatCurve=DataInterface() {propertyValues {Curve {keys [
          {arriveTangent=2.7533, interpMode=RCIM_Cubic, leaveTangent=2.847, tangentMode=RCTM_Break, time=0, value=0}
          {arriveTangent=1.3581, interpMode=RCIM_Cubic, leaveTangent=1.3581, tangentMode=RCTM_User, time=0.2558, value=0.5125}
          {arriveTangent=0.0529, interpMode=RCIM_Cubic, leaveTangent=0.2021, tangentMode=RCTM_User, time=1, value=1}
         ]}}})
    )
    Index 0 Param 4=&FloatFromCurve(FloatCurve=DataInterface() {propertyValues {Curve {keys [<at 0 1>,<at 1 0>]}}})
    Index 1 Param 4=1
    Index 2 Param 1=&FloatFromCurve(FloatCurve=DataInterface() {propertyValues {Curve {keys [
         {arriveTangent=2.646, interpMode=RCIM_Cubic, leaveTangent=2.646, tangentMode=RCTM_User, time=0, value=0}
         {leaveTangent=-0.0719, time=1, value=1}
        ]}}})
    Index 2 Param 2=&FloatFromCurve(
     FloatCurve=DataInterface() {propertyValues {Curve {keys [
         {arriveTangent=1.9056, time=0, value=0.2}
         {arriveTangent=0.0259, interpMode=RCIM_Cubic, leaveTangent=0.0259, tangentMode=RCTM_User, time=1, value=1.4}
        ]}}}
     Scale Curve=&RandomRangeFloat(Minimum=0.5, Maximum=1)
    )
    Index 2 Param 3=&FloatFromCurve(FloatCurve=DataInterface() {propertyValues {Curve {keys [
         {interpMode=RCIM_Cubic, leaveTangent=2.3064, tangentMode=RCTM_Break, time=0, value=1}
         {interpMode=RCIM_Cubic, leaveTangent=11.1233, tangentMode=RCTM_Break, time=1, value=2}
        ]}}})
    Index 2 Param 4=&FloatFromCurve(
     FloatCurve=DataInterface() {propertyValues {Curve {keys [
         {arriveTangent=-0.6867, interpMode=RCIM_Cubic, leaveTangent=-0.6867, tangentMode=RCTM_User, time=0, value=1}
         {interpMode=RCIM_Cubic, leaveTangent=-1.2956, tangentMode=RCTM_Break, time=1, value=0.75}
        ]}}}
     Scale Curve=1.6
    )
    Module.Index0Param0Write=true
    Module.Index0Param1Write=false
    Module.Index0Param2Write=true
    Module.Index0Param3Write=true
    Module.Index1Param0Write=true
    Module.Index1Param1Write=true
    Module.Index1Param2Write=true
    Module.Index1Param3Write=true
    Module.Index2Param0Write=true
    Module.Index2Param1Write=true
    Module.Index2Param2Write=true
    Module.Index2Param3Write=true
    Write Parameter Index 1=true
    Write Parameter Index 2=true
   )
   &ScaleColor(
    Scale Alpha=&MaskFloatBySpawnGroup(
     Masked Float=&FloatFromCurve(FloatCurve=DataInterface() {propertyValues {Curve {keys [
          {interpMode=RCIM_Cubic, tangentMode=RCTM_User, time=0, value=1}
          {arriveTangent=-1, interpMode=RCIM_Cubic, leaveTangent=-1, time=0.3521, value=0.3455}
          <at 1 0>
         ]}}})
     Passthrough Float=&MaskFloatBySpawnGroup(
      Masked Float=1
      Passthrough Float=&MaskFloatBySpawnGroup(
       Masked Float=&FloatFromCurve(FloatCurve=DataInterface() {propertyValues {Curve {keys [
            {interpMode=RCIM_Cubic, tangentMode=RCTM_User, time=0, value=1}
            {arriveTangent=-1, interpMode=RCIM_Cubic, leaveTangent=-1, time=0.5, value=0.35}
            <at 1 0>
           ]}}})
       Passthrough Float=&MaskFloatBySpawnGroup(
        Masked Float=&FloatFromCurve(FloatCurve=DataInterface() {propertyValues {Curve {keys [
             {interpMode=RCIM_Cubic, tangentMode=RCTM_User, time=0, value=1}
             {arriveTangent=-1, interpMode=RCIM_Cubic, leaveTangent=-1, time=0.3521, value=0.3455}
             <at 1 0>
            ]}}})
        Passthrough Float=&FloatFromCurve(FloatCurve=DataInterface() {propertyValues {Curve {keys [
             {interpMode=RCIM_Cubic, tangentMode=RCTM_User, time=0, value=0}
             {arriveTangent=1.25, interpMode=RCIM_Cubic, leaveTangent=1.25, time=0.2, value=1}
             {interpMode=RCIM_Cubic, tangentMode=RCTM_User, time=0.8, value=1}
             {tangentMode=RCTM_User, time=1, value=0}
            ]}}})
        Spawn Group Mask=3
       )
       Spawn Group Mask=2
      )
      Spawn Group Mask=1
     )
    )
    Scale RGB=&MaskVectorBySpawnGroup(
     Masked Vector=&VectorFromFloat(Value=&FloatFromCurve(FloatCurve=DataInterface() {propertyValues {Curve {keys [{arriveTangent=1.8145, interpMode=RCIM_Cubic, leaveTangent=0.8876, tangentMode=RCTM_Break, time=0, value=0.75}, <at 0.3387 0.9565>, <at 1 1>]}}}))
     Passthrough Vector=&MaskVectorBySpawnGroup(
      Masked Vector=&VectorFromFloat(Value=&FloatFromCurve(FloatCurve=DataInterface() {propertyValues {Curve {keys [{arriveTangent=1.8145, interpMode=RCIM_Cubic, leaveTangent=0.8876, tangentMode=RCTM_Break, time=0, value=0.75}, <at 0.3387 0.9565>, <at 1 1>]}}}))
      Passthrough Vector=&MaskVectorBySpawnGroup(Masked Vector=<0 0 0>, Passthrough Vector=<1 1 1>, Spawn Group Mask=4)
      Spawn Group Mask=3
     )
    )
    ScaleA=true
    ScaleRGB=true
   )
   &SkeletalMeshLocation(
    Bone / Socket Index=3
    MaskSpawnGroup=true
    Mesh Sampling Type=::NewEnumerator5
    Skeletal Mesh=$System.SkeletalMesh
    Socket Sampling Mode=::NewEnumerator5
    Spawn Group Mask=4
   ) {module=SkeletalMeshLocation001}
  ]
  rendererClasses [&NiagaraSpriteRendererProperties]
  renderers [
   {rendererClass=&NiagaraSpriteRendererProperties}
   {rendererClass=&NiagaraSpriteRendererProperties, rendererIndex=1}
   {rendererClass=&NiagaraSpriteRendererProperties, rendererIndex=2}
   {rendererClass=&NiagaraSpriteRendererProperties, rendererIndex=3}
   {rendererClass=&NiagaraSpriteRendererProperties, rendererIndex=4}
  ]
 }
 Exhaust_and_ShiftBlip {
  emitterUpdateScript [
   &EmitterState(Life Cycle Mode=::NewEnumerator1, Loop Duration=0.1, UseLoopDelay=false)
   &SpawnRate(
    SpawnRate=&MakeCustomFloatFromBool(
     Bool=&Boolean_AndOperation(
      Bool1=$User.VehicleHasDriver
      Bool2=&Boolean_AndOperation(Bool1=$User.VehicleHasFuel, Bool2=&Boolean_NotOperation(Bool=$User.IsWaterDeep))
     )
     True Float=&Multiply_Float(A=&Lerp_Float(A=12, B=90, Alpha=$User.ChargeProgress), B=$System.Timeline001.Float)
     False Float=0
    )
    Use Spawn Probability=false
   )
   &SpawnRate(SpawnGroup=1, SpawnRate=&Multiply_Float(A=80, B=$System.Timeline.Float), Use Spawn Probability=false) {module=SpawnRate001}
   &SpawnBurst_Instantaneous(
    Spawn Count=&RandomRangeInt(Minimum=1, Maximum=2)
    Spawn Group=2
    Spawn Probability=&MakeCustomFloatFromBool(
     Bool=&Boolean_AndOperation(
      Bool1=$User.VehicleHasDriver
      Bool2=&Boolean_AndOperation(
       Bool1=$User.VehicleHasFuel
       Bool2=&Boolean_AndOperation(Bool1=$User.IsShifting, Bool2=&Boolean_NotOperation(Bool=$User.IsWaterDeep))
      )
     )
     True Float=1
     False Float=0
    )
    Use Spawn Probability=true
    UseLoopCountLimit=false
   )
   &SpawnBurst_Instantaneous(
    Spawn Group=3
    Spawn Probability=&MakeCustomFloatFromBool(
     Bool=&Boolean_AndOperation(
      Bool1=$User.VehicleHasDriver
      Bool2=&Boolean_AndOperation(
       Bool1=$User.VehicleHasFuel
       Bool2=&Boolean_AndOperation(Bool1=$User.IsShifting, Bool2=&Boolean_NotOperation(Bool=$User.IsWaterDeep))
      )
     )
     True Float=1
     False Float=0
    )
    Use Spawn Probability=true
    UseLoopCountLimit=false
   ) {module=SpawnBurst_Instantaneous001}
  ]
  particleSpawnScript [
   /SetVariables(Particles.VisibilityTag=&SpawnGroup()) {isSetParameters=true}
   &InitializeParticle(
    Color=&MaskLinearColorBySpawnGroup(Masked Linear Color=<rgba 0.4236 0.4777 0.5469 0.025>, Passthrough Linear Color=<rgba 1 1 1 1>)
    Lifetime=&MaskFloatBySpawnGroup(
     Masked Float=&Lerp_Float(
      A=&Lerp_Float(A=&RandomRangeFloat(Minimum=0.225, Maximum=0.8), B=1, Alpha=$User.Vehicle_Speed)
      B=0.2
      Alpha=$User.ChargeProgress
     )
     Passthrough Float=&MaskFloatBySpawnGroup(
      Masked Float=&RandomRangeFloat(Minimum=0.225, Maximum=0.375)
      Passthrough Float=&MaskFloatBySpawnGroup(
       Masked Float=&RandomRangeFloat(Minimum=0.15, Maximum=0.25)
       Passthrough Float=&RandomRangeFloat(Minimum=0.085, Maximum=0.125)
       Spawn Group Mask=2
      )
      Spawn Group Mask=1
     )
    )
    Sprite Rotation Mode=::NewEnumerator2
    Sprite Size Mode=::NewEnumerator3
    Uniform Sprite Size=&Lerp_Float(
     A=&RandomRangeFloat(Minimum=40, Maximum=60)
     B=&MaskFloatBySpawnGroup(
      Masked Float=&RandomRangeFloat(Minimum=50, Maximum=65)
      Passthrough Float=&RandomRangeFloat(Minimum=60, Maximum=80)
     )
     Alpha=&Abs_Float(Float:NiagaraFloat=$User.Vehicle_Speed)
    )
    UsePositionOffset=false
    Write Material Random=true
   )
   &CameraOffset(Camera Offset Amount=&MaskFloatBySpawnGroup(Masked Float=10, Passthrough Float=0))
  ]
  particleUpdateScript [
   &ParticleState()
   &ScaleSpriteSize(
    Scale Sprite Size Mode=::NewEnumerator0
    Uniform Scale Factor=&MaskFloatBySpawnGroup(
     Masked Float=&FloatFromCurve(FloatCurve=DataInterface() {propertyValues {Curve {keys [{interpMode=RCIM_Cubic, tangentMode=RCTM_Break, time=0, value=0.179}, <at 0.8329 1>]}}})
     Passthrough Float=&FloatFromCurve(
      FloatCurve=DataInterface() {propertyValues {Curve {keys [{interpMode=RCIM_Cubic, leaveTangent=1.3262, tangentMode=RCTM_Break, time=0, value=0.179}, <at 0.8329 1>]}}}
      Scale Curve=&MaskFloatBySpawnGroup(Masked Float=1, Passthrough Float=0.8, Spawn Group Mask=1)
     )
    )
   )
   &SpriteRotationRate(Rotation Rate=$Particles.Initial.MaterialRandom)
   &SkeletalMeshLocation(
    Bone / Socket Index=4
    MaskSpawnGroup=false
    Mesh Sampling Type=::NewEnumerator5
    Skeletal Mesh=$System.SkeletalMesh
    Socket Sampling Mode=::NewEnumerator5
   )
   &ScaleColor(
    Scale Alpha=&FloatFromCurve(
     FloatCurve=DataInterface() {propertyValues {Curve {keys [{interpMode=RCIM_Cubic, tangentMode=RCTM_User, time=0, value=1}, <at 1 0>]}}}
     Scale Curve=&Lerp_Float(
      A=8
      B=30
      Alpha=&Add_Float(A=$User.Vehicle_Speed, B=&Multiply_Float(A=$User.ChargeProgress, B=0.5))
     )
    )
    Scale RGB=&VectorFromFloat(Value=&FloatFromCurve(FloatCurve=DataInterface() {propertyValues {Curve {keys [{arriveTangent=1.8145, interpMode=RCIM_Cubic, leaveTangent=4.5795, tangentMode=RCTM_Break, time=0, value=0.4}, <at 0.3387 0.9565>, <at 1 1>]}}}))
    ScaleA=true
    ScaleRGB=true
   )
   /SetVariables(
    Particles.Flame Color:LinearColor=&ColorFromCurve(ColorCurve=DataInterface() {propertyValues {
       AlphaCurve {keys [<at 0.3104 1 RCIM_Cubic>, {arriveTangent=-1.4502, interpMode=RCIM_Cubic, leaveTangent=-1.4502, time=0.4, value=0.3}, <at 1 0 RCIM_Cubic>]}
       BlueCurve {keys [<at 0.0194 50 RCIM_Cubic>,<at 0.147 67.3043>,<at 0.1866 2.1967>,<at 0.3189 0 RCIM_Cubic>]}
       GreenCurve {keys [<at 0.0194 30.8898 RCIM_Cubic>,<at 0.147 8.786>,<at 0.1866 11.2076>,<at 0.3189 5.5901 RCIM_Cubic>]}
       RedCurve {keys [<at 0.0194 0.1 RCIM_Cubic>,<at 0.147 150>,<at 0.1866 50>,<at 0.3189 25 RCIM_Cubic>]}
      }})
    Particles.Flame Dynamic Parameters:Vector4f=&MakeVector4(
     X=&FloatFromCurve(FloatCurve=DataInterface() {propertyValues {Curve {keys [
          {arriveTangent=2.3188, interpMode=RCIM_Cubic, time=0, value=0}
          {arriveTangent=2, interpMode=RCIM_Cubic, leaveTangent=2, time=0.2449, value=0.8818}
          <at 0.5 1 RCIM_Cubic>
         ]}}})
     Y=&RandomRangeFloat(Minimum=0.333, Maximum=1)
     Z=&FloatFromCurve(FloatCurve=DataInterface() {propertyValues {Curve {keys [{arriveTangent=1.4224, interpMode=RCIM_Cubic, time=0.5, value=1}, <at 1 2 RCIM_Cubic>]}}})
     W=&RandomRangeFloat(Minimum=1.333, Maximum=2)
    )
    Particles.FlameSubUVIndex:NiagaraFloat=&Add_Float(
     A=&RandomRangeFloat(Minimum=0, Maximum=32)
     B=&Lerp_Float(A=0, B=60, Alpha=$Particles.NormalizedAge)
    )
   ) {isSetParameters=true}
   /SetVariables(
    Particles.LightColor=&ColorFromCurve(ColorCurve=DataInterface() {propertyValues {
       AlphaCurve {keys [<at 0.0118 0>,<at 0.3542 1 RCIM_Cubic>,<at 1 0 RCIM_Cubic>]}
       BlueCurve {keys [<at 0.054 0>,<at 0.4426 0 RCIM_Cubic>]}
       GreenCurve {keys [<at 0.054 4.7963>,<at 0.4426 2.6832 RCIM_Cubic>]}
       RedCurve {keys [<at 0.054 24>,<at 0.4426 12 RCIM_Cubic>]}
      }})
    Particles.LightExponent=8
    Particles.LightRadius=120
   ) {isSetParameters=true}
  ]
  rendererClasses [&NiagaraSpriteRendererProperties, &NiagaraLightRendererProperties]
  renderers [
   {rendererClass=&NiagaraSpriteRendererProperties}
   {rendererClass=&NiagaraSpriteRendererProperties, rendererIndex=1}
   {rendererClass=&NiagaraSpriteRendererProperties, rendererIndex=2}
   {rendererClass=&NiagaraSpriteRendererProperties, rendererIndex=3}
   {rendererClass=&NiagaraLightRendererProperties, rendererIndex=4}
  ]
 }
 NE_Parent_MudTrail {
  emitterUpdateScript [
   &EmitterState()
   &SpawnPerUnit(
    Module.Use Max Movement Threshold=false
    Module.Use Movement Tolerance=false
    Spawn Per Unit=10
    Spawn Probability:NiagaraFloat=&MakeCustomFloatFromBool(Bool=$System.SpawnMud, True Float=1, False Float=0)
    Use Spawn Probability=true
   )
  ]
  particleSpawnScript [
   &PartitionParticles()
   &InitializeParticle(
    Lifetime=&SelectFloatFromArray(
     Direct Array Index=$Particles.PartitionParticles.Partition
     Float Selection Array=DataInterface() {propertyValues {FloatData [1.7,2.7,2.5,1.5]}}
    )
    Mass Mode=::NewEnumerator1
    Sprite Size=<25 25>
    Sprite Size Mode=::NewEnumerator4
    UsePositionOffset=false
    Write Material Random=true
   ) {module=InitializeParticle001}
   /SetVariables() {isSetParameters=true, module=SetVariables_#5}
   /SetVariables(Particles.RibbonWidth=&SelectFloatFromArray(Direct Array Index=$Particles.PartitionParticles.Partition, Float Selection Array=DataInterface() {propertyValues {FloatData [21,15]}})) {isSetParameters=true, module=SetVariables_#6}
   &SkeletalMeshLocation(
    Bone / Socket Index=&Add_Integer(A=1, B=$Particles.PartitionParticles.Partition)
    MaskSpawnGroup=false
    Mesh Sampling Type=::NewEnumerator5
    Sampled Position Offset=&MakeCustomVectorFromBool(
     Bool=&SetBoolByFloatComparison(A=0.5, B=&MakeFloatFromInt(Int=$Particles.PartitionParticles.Partition))
     True Vector=<0 0 0>
     False Vector=<0 0 -20>
    )
    Sampled Position Offset Coordinate Space:ENiagaraSkelMeshTransforms=::NewEnumerator1
    Skeletal Mesh=$System.SkeletalMesh
    Socket Sampling Mode=::NewEnumerator5
   )
   &DynamicMaterialParameters(
    Index 0 Param 1=&SelectFloatFromArray(
     Direct Array Index=$Particles.PartitionParticles.Partition
     Float Selection Array=DataInterface() {propertyValues {FloatData [0,0.25,0.5,1]}}
    )
    Module.Param0WriteEnabled=true
    Module.Param1WriteEnabled=true
    Module.Param2WriteEnabled=true
    Module.Param3WriteEnabled=true
   )
  ]
  particleUpdateScript [
   &ParticleState()
   &SolveForcesAndVelocity(Clamp Velocity=false, Limit Acceleration=false)
   &CameraOffset(Camera Offset Amount=32)
   &ScaleColor(
    Scale Alpha=&FloatFromCurve(FloatCurve=DataInterface() {propertyValues {Curve {keys [
         <at 0 1 RCIM_Cubic>
         {arriveTangent=-1.3333, interpMode=RCIM_Cubic, leaveTangent=-1.3333, time=0.2965, value=0.8703}
         {arriveTangent=-1.237, interpMode=RCIM_Cubic, leaveTangent=-1.237, time=0.6, value=0.2}
         <at 1 0>
        ]}}})
    ScaleA=true
    ScaleRGB=true
   )
   &ScaleRibbonWidth(Ribbon Width Scale=&FloatFromCurve(FloatCurve=DataInterface() {propertyValues {Curve {keys [<at 0 0.5>,<at 0.0172 1>,<at 0.6885 1>]}}}))
  ]
  rendererClasses [&NiagaraRibbonRendererProperties]
  renderers [{rendererClass=&NiagaraRibbonRendererProperties}]
 }
 PeelOut {
  emitterUpdateScript [
   &EmitterState()
   &SpawnRate(
    SpawnRate=&MakeCustomFloatFromBool(
     Bool=&Boolean_AndOperation(
      Bool1=$User.PeelOut_Wheel_Rear
      Bool2=&Boolean_NotOperation(Bool=&Boolean_OrOperation(Bool1=$User.IsWaterDeep, Bool2=&Boolean_OrOperation(Bool1=$System.SpawnMud, Bool2=$User.IsDoingStoppie)))
     )
     True Float=&Multiply_Float(
      A=85
      B=&Lerp_Float(
       A=&MakeCustomFloatFromBool(Bool=$User.IsPeelOutOnRoad, True Float=0, False Float=1)
       B=1
       Alpha=$System.Timeline002.Float
      )
     )
     False Float=0
    )
    Use Spawn Probability=false
   )
   &SpawnRate(
    SpawnGroup=1
    SpawnRate=&MakeCustomFloatFromBool(
     Bool=&Boolean_AndOperation(
      Bool1=&Boolean_AndOperation(Bool1=$User.PeelOut_Wheel_Rear, Bool2=$System.SpawnDirtClods)
      Bool2=&Boolean_NotOperation(Bool=&Boolean_OrOperation(Bool1=$User.IsWaterDeep, Bool2=$User.IsDoingStoppie))
     )
     True Float=90
     False Float=0
    )
    Use Spawn Probability=false
   ) {module=SpawnRate002}
   &SpawnRate(
    SpawnGroup=2
    SpawnRate=&MakeCustomFloatFromBool(
     Bool=&Boolean_AndOperation(
      Bool1=&Boolean_AndOperation(Bool1=$User.PeelOut_Wheel_Rear, Bool2=$System.SpawnGrass)
      Bool2=&Boolean_NotOperation(Bool=&Boolean_OrOperation(Bool1=$User.IsWaterDeep, Bool2=$User.IsDoingStoppie))
     )
     True Float=90
     False Float=0
    )
    Use Spawn Probability=false
   ) {module=SpawnRate003}
   &SpawnRate(
    SpawnGroup=3
    SpawnRate=&MakeCustomFloatFromBool(
     Bool=&Boolean_AndOperation(
      Bool1=&Boolean_AndOperation(Bool1=$User.PeelOut_Wheel_Rear, Bool2=$System.SpawnSand)
      Bool2=&Boolean_NotOperation(Bool=&Boolean_OrOperation(Bool1=$User.IsWaterDeep, Bool2=$User.IsDoingStoppie))
     )
     True Float=&Lerp_Float(
      A=&MakeCustomFloatFromBool(Bool=$User.IsPeelOutOnRoad, True Float=0, False Float=30)
      B=30
      Alpha=$System.Timeline002.Float
     )
     False Float=0
    )
    Use Spawn Probability=false
   ) {module=SpawnRate004}
   &SpawnRate(
    SpawnGroup=4
    SpawnRate=&MakeCustomFloatFromBool(
     Bool=&Boolean_AndOperation(
      Bool1=&Boolean_AndOperation(Bool1=$User.PeelOut_Wheel_Rear, Bool2=$System.SpawnSnow)
      Bool2=&Boolean_NotOperation(Bool=&Boolean_OrOperation(Bool1=$User.IsWaterDeep, Bool2=$User.IsDoingStoppie))
     )
     True Float=&Lerp_Float(
      A=&MakeCustomFloatFromBool(Bool=$User.IsPeelOutOnRoad, True Float=0, False Float=25)
      B=25
      Alpha=$System.Timeline002.Float
     )
     False Float=0
    )
    Use Spawn Probability=false
   ) {module=SpawnRate005}
   &SpawnRate(
    SpawnGroup=5
    SpawnRate=&MakeCustomFloatFromBool(
     Bool=&Boolean_AndOperation(
      Bool1=&Boolean_AndOperation(Bool1=$User.PeelOut_Wheel_Rear, Bool2=$System.SpawnSnow)
      Bool2=&Boolean_NotOperation(Bool=&Boolean_OrOperation(Bool1=$User.IsWaterDeep, Bool2=$User.IsDoingStoppie))
     )
     True Float=&Lerp_Float(
      A=&MakeCustomFloatFromBool(Bool=$User.IsPeelOutOnRoad, True Float=0, False Float=60)
      B=60
      Alpha=$System.Timeline002.Float
     )
     False Float=0
    )
    Use Spawn Probability=false
   ) {module=SpawnRate006}
   &SpawnRate(
    SpawnGroup=7
    SpawnRate=&MakeCustomFloatFromBool(
     Bool=&Boolean_AndOperation(
      Bool1=&Boolean_AndOperation(Bool1=$User.PeelOut_Wheel_Rear, Bool2=$System.SpawnIce)
      Bool2=&Boolean_NotOperation(Bool=&Boolean_OrOperation(Bool1=$User.IsWaterDeep, Bool2=$User.IsDoingStoppie))
     )
     True Float=25
     False Float=0
    )
    Use Spawn Probability=false
   ) {module=SpawnRate007}
   &SpawnRate(
    SpawnGroup=8
    SpawnRate=&MakeCustomFloatFromBool(
     Bool=&Boolean_AndOperation(
      Bool1=&Boolean_AndOperation(Bool1=$User.PeelOut_Wheel_Rear, Bool2=$System.SpawnIce)
      Bool2=&Boolean_NotOperation(Bool=&Boolean_OrOperation(Bool1=$User.IsWaterDeep, Bool2=$User.IsDoingStoppie))
     )
     True Float=80
     False Float=0
    )
    Use Spawn Probability=false
   ) {module=SpawnRate008}
   &SpawnRate(
    SpawnGroup=9
    SpawnRate=&MakeCustomFloatFromBool(
     Bool=&Boolean_AndOperation(
      Bool1=&Boolean_AndOperation(Bool1=$User.PeelOut_Wheel_Rear, Bool2=$System.SpawnMud)
      Bool2=&Boolean_NotOperation(Bool=&Boolean_OrOperation(Bool1=$User.IsWaterDeep, Bool2=$User.IsDoingStoppie))
     )
     True Float=&Lerp_Float(
      A=&MakeCustomFloatFromBool(Bool=$User.IsPeelOutOnRoad, True Float=0, False Float=250)
      B=250
      Alpha=$System.Timeline002.Float
     )
     False Float=0
    )
    Use Spawn Probability=false
   ) {module=SpawnRate009}
   &SpawnRate(
    SpawnGroup=10
    SpawnRate=&MakeCustomFloatFromBool(
     Bool=&Boolean_AndOperation(
      Bool1=&Boolean_AndOperation(Bool1=$User.PeelOut_Wheel_Rear, Bool2=$System.SpawnMud)
      Bool2=&Boolean_NotOperation(Bool=&Boolean_OrOperation(Bool1=$User.IsWaterDeep, Bool2=$User.IsDoingStoppie))
     )
     True Float=&Lerp_Float(
      A=&MakeCustomFloatFromBool(Bool=$User.IsPeelOutOnRoad, True Float=0, False Float=110)
      B=110
      Alpha=$System.Timeline002.Float
     )
     False Float=0
    )
    Use Spawn Probability=false
   ) {module=SpawnRate010}
   &SpawnRate(
    SpawnGroup=11
    SpawnRate=&MakeCustomFloatFromBool(
     Bool=&Boolean_AndOperation(
      Bool1=&Boolean_AndOperation(Bool1=$User.PeelOut_Wheel_Rear, Bool2=$System.SpawnMud)
      Bool2=&Boolean_NotOperation(Bool=&Boolean_OrOperation(Bool1=$User.IsWaterDeep, Bool2=$User.IsDoingStoppie))
     )
     True Float=220
     False Float=0
    )
    Use Spawn Probability=false
   ) {module=SpawnRate011}
   &SpawnRate(
    SpawnGroup=12
    SpawnRate=&MakeCustomFloatFromBool(
     Bool=&Boolean_AndOperation(
      Bool1=$User.PeelOut_Wheel_Rear
      Bool2=&Boolean_NotOperation(Bool=&Boolean_OrOperation(Bool1=$System.SpawnMud, Bool2=&Boolean_OrOperation(Bool1=$User.IsWaterDeep, Bool2=$User.IsDoingStoppie)))
     )
     True Float=126
     False Float=0
    )
    Use Spawn Probability=false
   ) {module=SpawnRate001}
  ]
  particleSpawnScript [
   /SetVariables(Particles.VisibilityTag=$Emitter.SpawnGroup) {isSetParameters=true}
   &InitializeParticle(
    Color=&self:SpawnGroupColorSelection(
     00_Smoke:LinearColor=$System.PeelOutColor
     01_Dirt Clods:LinearColor=&MultiplyLinearColor(Linear Color=$System.PeelOutColor, Scale Factor=<1 1 1 1.2>)
     02_Grass:LinearColor=<rgba 0.5737 1 0.1872 1>
     03_Sand:LinearColor=$System.PeelOutColor
     04_SnowMist:LinearColor=<rgba 0.6919 0.7584 0.8177 0.3>
     05_SnowChunks:LinearColor=<rgba 1 1 1 1>
     07_IceMesh:LinearColor=<rgba 0.1337 0.2213 0.276 1>
     08_IceShinyFlakes:LinearColor=&RandomRangeLinearColor(Minimum=<rgba 1 1 1 1>, Maximum=<rgba 0.1036 0.4193 1 1>)
     09_Mud Kickup:LinearColor=<rgba 0.1406 0.0594 0.0253 1>
     10_MudSplash:LinearColor=<rgba 0.2031 0.0555 0.0271 1>
     11_Mud PeelOut Splash:LinearColor=$System.PeelOutColor
     12_Peel Out Spikes:LinearColor=$System.PeelOutColor
    )
    Lifetime=&self:SpawnGroupFloatSelection(
     00_Smoke:NiagaraFloat=&Lerp_Float(
      A=&MakeCustomFloatFromBool(
       Bool=$System.SpawnSand
       True Float=&RandomRangeFloat(Minimum=1.5, Maximum=2)
       False Float=&RandomRangeFloat(Minimum=0.6, Maximum=0.8)
      )
      B=1.3
      Alpha=&Abs_Float(Float:NiagaraFloat=$User.Vehicle_Speed)
     )
     01_DirtClods:NiagaraFloat=&Lerp_Float(A=&UniformRangedFloat(Minimum=1.5, Maximum=2.5), B=1, Alpha=$User.Vehicle_Speed)
     02_Grass:NiagaraFloat=&RandomRangeFloat(Minimum=0.5, Maximum=1.5)
     03_Sand:NiagaraFloat=&Lerp_Float(
      A=&RandomRangeFloat(Minimum=1.3, Maximum=1.5)
      B=1.3
      Alpha=&Abs_Float(Float:NiagaraFloat=$User.Vehicle_Speed)
     )
     04_SnowMist:NiagaraFloat=&RandomRangeFloat(Minimum=0.7, Maximum=1.2)
     05_SnowChunks:NiagaraFloat=&RandomRangeFloat(Minimum=1.5, Maximum=2)
     07_Ice Mesh:NiagaraFloat=0.7
     08_Ice Shiny Flakes:NiagaraFloat=&RandomRangeFloat(Minimum=0.5, Maximum=1)
     09_Mud Kickup:NiagaraFloat=&RandomRangeFloat(Minimum=0.5, Maximum=0.8)
     10_Mud Splash:NiagaraFloat=&RandomRangeFloat(Minimum=0.5, Maximum=0.8)
     11_Mud Peel Out Splash:NiagaraFloat=&RandomRangeFloat(Minimum=0.11, Maximum=0.14)
     12_Peel Out Spikes:NiagaraFloat=&RandomRangeFloat(Minimum=0.11, Maximum=0.14)
    )
    Mesh Renderer Array Visibility Mode=::NewEnumerator3
    Mesh Renderer Info=DataInterface() {propertyValues {MeshRenderer=&self:PeelOut_1.NiagaraMeshRendererProperties_0}}
    Mesh Scale=&self:SpawnGroupVectorSelection(
     00_Smoke:Vector3f=<0 0 0>
     01_DirtClods:Vector3f=<0 0 0>
     02_Grass:Vector3f=&RandomRangeVector(Minimum=<0.1 0.2 0.2>, Maximum=<0.3 0.5 0.5>)
     03_Sand:Vector3f=<0 0 0>
     04_SnowMist:Vector3f=<0 0 0>
     05_SnowChunks:Vector3f=<0 0 0>
     07_IceMesh:Vector3f=&Multiply_VectorByFloat(Vector=&RandomRangeVector(Minimum=<2 0.5 0.5>, Maximum=<3 1.5 1.5>), Float=0.1)
     08_IceShinyFlakes:Vector3f=<0 0 0>
     09_MudKickup:Vector3f=<0 0 0>
     10_MudSplash:Vector3f=<0 0 0>
     11_Mud PeelOut Splash:Vector3f=<0 0 0>
     12_Peel Out Spikes:Vector3f=<0 0 0>
    )
    Mesh Scale Mode=::NewEnumerator4
    Sprite Rotation Mode=::NewEnumerator2
    Sprite Size=&self:SpawnGroupVector2DSelection(
     00_Smoke:Vector2f=&Vector2DFromFloat(Value=&Lerp_Float(
       A=&RandomRangeFloat(Minimum=40, Maximum=60)
       B=&RandomRangeFloat(Minimum=65, Maximum=80)
       Alpha=&Abs_Float(Float:NiagaraFloat=$User.Vehicle_Speed)
      ))
     01_DirtClods:Vector2f=&Vector2DFromFloat(Value=&Lerp_Float(
       A=&UniformRangedFloat(Minimum=40, Maximum=60)
       B=&UniformRangedFloat(Minimum=65, Maximum=80)
       Alpha=&Abs_Float(Float:NiagaraFloat=$User.Vehicle_Speed)
      ))
     02_Grass:Vector2f=<0 0>
     03_Sand:Vector2f=&Vector2DFromFloat(Value=&Lerp_Float(
       A=&RandomRangeFloat(Minimum=40, Maximum=60)
       B=&RandomRangeFloat(Minimum=65, Maximum=80)
       Alpha=&Abs_Float(Float:NiagaraFloat=$User.Vehicle_Speed)
      ))
     04_SnowMist:Vector2f=&Vector2DFromFloat(Value=&RandomRangeFloat(Minimum=250, Maximum=300))
     05_SnowChunks:Vector2f=&Vector2DFromFloat(Value=&RandomRangeFloat(Minimum=5, Maximum=9))
     07_IceMesh:Vector2f=<1 1>
     08_IceShinyFlakes:Vector2f=&Vector2DFromFloat(Value=&RandomRangeFloat(Minimum=5, Maximum=15))
     09_Mud Kickup:Vector2f=&Vector2DFromFloat(Value=&RandomRangeFloat(Minimum=140, Maximum=200))
     10_Mud Splash:Vector2f=&Vector2DFromFloat(Value=&RandomRangeFloat(Minimum=40, Maximum=60))
     11_Mud PeelOut Splash:Vector2f=&MakeVector2D(
      X=&RandomRangeFloat(Minimum=15, Maximum=25)
      Y=&Lerp_Float(
       A=&RandomRangeFloat(Minimum=35, Maximum=70)
       B=&RandomRangeFloat(Minimum=100, Maximum=180)
       Alpha=&Abs_Float(Float:NiagaraFloat=$User.Vehicle_Speed)
      )
     )
     12_PeelOut Spikes:Vector2f=&MakeVector2D(
      X=&RandomRangeFloat(Minimum=15, Maximum=25)
      Y=&Lerp_Float(
       A=&RandomRangeFloat(Minimum=35, Maximum=70)
       B=&RandomRangeFloat(Minimum=100, Maximum=180)
       Alpha=&Abs_Float(Float:NiagaraFloat=$User.Vehicle_Speed)
      )
     )
    )
    Sprite Size Mode=::NewEnumerator4
    UsePositionOffset=false
    Write Material Random=true
   )
   &SkeletalMeshLocation(
    Bone / Socket Index=1
    MaskSpawnGroup=false
    Mesh Sampling Type=::NewEnumerator5
    Skeletal Mesh=$System.SkeletalMesh
    Socket Sampling Mode=::NewEnumerator5
   )
   &ShapeLocation(MaskSpawnGroup=true, Override Seed=false, Spawn Group Mask=5, Sphere Radius=18, UseOwnerScale=true)
   &ShapeLocation(MaskSpawnGroup=true, Override Seed=false, Spawn Group Mask=9, Sphere Radius=18, UseOwnerScale=true) {module=ShapeLocation001}
   &ShapeLocation(MaskSpawnGroup=true, Override Seed=false, Spawn Group Mask=10, Sphere Radius=18, UseOwnerScale=true) {module=ShapeLocation002}
   &AddVelocity(Velocity=&self:SpawnGroupVectorSelection(
     00_Smoke:Vector3f=&RandomRangeVector(
      Minimum=<500 -150 100>
      Maximum=&MakeVector(
       X=600
       Y=150
       Z=&Lerp_Float(
        A=&MakeCustomFloatFromBool(Bool=$User.IsPeelOutOnRoad, True Float=0, False Float=500)
        B=1400
        Alpha=$System.Timeline002.Float
       )
      )
     )
     01_DirtClods:Vector3f=&Multiply_VectorByFloat(
      Vector=&RandomRangeVector(Minimum=<500 -50 50>, Maximum=<600 50 1200>)
      Float=&Lerp_Float(A=0.2, B=1, Alpha=$System.Timeline002.Float)
     )
     02_Grass:Vector3f=&RandomRangeVector(Minimum=<500 -50 50>, Maximum=<600 50 1200>)
     03_Sand:Vector3f=&RandomRangeVector(
      Minimum=<500 -150 100>
      Maximum=&MakeVector(
       X=600
       Y=150
       Z=&Lerp_Float(
        A=&MakeCustomFloatFromBool(Bool=$User.IsPeelOutOnRoad, True Float=0, False Float=500)
        B=1400
        Alpha=$System.Timeline002.Float
       )
      )
     )
     04_SnowMist:Vector3f=&RandomRangeVector(
      Minimum=<500 -150 100>
      Maximum=&MakeVector(
       X=600
       Y=150
       Z=&Lerp_Float(
        A=&MakeCustomFloatFromBool(Bool=$User.IsPeelOutOnRoad, True Float=0, False Float=500)
        B=1400
        Alpha=$System.Timeline002.Float
       )
      )
     )
     05_SnowChunks:Vector3f=&RandomRangeVector(
      Minimum=<20 -100 300>
      Maximum=&MakeVector(
       X=300
       Y=100
       Z=&Lerp_Float(
        A=&MakeCustomFloatFromBool(Bool=$User.IsPeelOutOnRoad, True Float=0, False Float=1200)
        B=1800
        Alpha=$System.Timeline002.Float
       )
      )
     )
     07_IceMesh:Vector3f=&Multiply_VectorByFloat(
      Vector=&RandomRangeVector(Minimum=<500 -50 50>, Maximum=<600 50 1200>)
      Float=&Lerp_Float(A=0.2, B=0.4, Alpha=$System.Timeline002.Float)
     )
     08_IceShinyFlakes:Vector3f=&RandomRangeVector(
      Minimum=<20 -100 300>
      Maximum=&MakeVector(
       X=300
       Y=100
       Z=&Lerp_Float(
        A=&MakeCustomFloatFromBool(Bool=$User.IsPeelOutOnRoad, True Float=0, False Float=500)
        B=1400
        Alpha=$System.Timeline002.Float
       )
      )
     )
     09_MudKickup:Vector3f=&RandomRangeVector(
      Minimum=<20 -100 300>
      Maximum=&MakeVector(
       X=300
       Y=100
       Z=&Lerp_Float(
        A=&MakeCustomFloatFromBool(Bool=$User.IsPeelOutOnRoad, True Float=0, False Float=1200)
        B=1800
        Alpha=$System.Timeline002.Float
       )
      )
     )
     10_MudSplash:Vector3f=&RandomRangeVector(
      Minimum=<20 -100 300>
      Maximum=&MakeVector(
       X=300
       Y=100
       Z=&Lerp_Float(
        A=&MakeCustomFloatFromBool(Bool=$User.IsPeelOutOnRoad, True Float=0, False Float=1200)
        B=1800
        Alpha=$System.Timeline002.Float
       )
      )
     )
     11_Mud PeelOut Splash:Vector3f=<0 0 0>
     12_Peel Out Spikes:Vector3f=<0 0 0>
    ))
   &AddVelocity(
    Cone Angle=27
    Cone Axis=&MakeVector(X=1, Y=&Multiply_Float(A=$User.Slide_Direction, B=-1), Z=0.5)
    Module.UseConeFalloff=false
    Module.UseVelDistribution=false
    Override Seed=false
    Velocity Mode=::NewEnumerator2
    Velocity Speed=&self:SpawnGroupFloatSelection(
     00_Smoke:NiagaraFloat=0
     01_DirtClods:NiagaraFloat=0
     02_Grass:NiagaraFloat=0
     03_Sand:NiagaraFloat=0
     04_SnowMist:NiagaraFloat=0
     05_SnowChunks:NiagaraFloat=0
     07_Ice Mesh:NiagaraFloat=0
     08_Ice Shiny Flakes:NiagaraFloat=0
     09_Mud Kickup:NiagaraFloat=0
     10_Mud Splash:NiagaraFloat=0
     11_Mud Peel Out Splash:NiagaraFloat=600
     12_Peel Out Spikes:NiagaraFloat=600
    )
   ) {module=AddVelocity001}
   &SubUVAnimation(
    Evaluation Type=::NewEnumerator0
    "Sprite Renderer "=DataInterface() {propertyValues {SpriteRenderer=&self:PeelOut_1.NiagaraSpriteRendererProperties_12}}
    SubUV Animation Mode:ENiagara_SubUVLookupModeV2=::NewEnumerator2
    UseEndFrame=false
    UseStartFrame=false
   )
   &InitialMeshOrientation(Use Rotation Vector=false)
   /SetVariables(Particles.IceMesh_Low_Color:LinearColor=<rgba 0.7756 0.9185 1 1>) {isSetParameters=true}
  ]
  particleUpdateScript [
   &ParticleState()
   &SkeletalMeshLocation(
    Bone / Socket Index=1
    MaskSpawnGroup=true
    Mesh Sampling Type=::NewEnumerator5
    Skeletal Mesh=$System.SkeletalMesh
    Socket Sampling Mode=::NewEnumerator5
    Spawn Group Mask=12
   ) {module=SkeletalMeshLocation001}
   &SkeletalMeshLocation(
    Bone / Socket Index=&RandomRangeInt(Minimum=0, Maximum=1)
    MaskSpawnGroup=true
    Mesh Sampling Type=::NewEnumerator5
    Skeletal Mesh=$System.SkeletalMesh
    Socket Sampling Mode=::NewEnumerator5
    Spawn Group Mask=11
   ) {module=SkeletalMeshLocation002}
   &ScaleColor(
    Scale Alpha=&self:SpawnGroupFloatSelection(
     00_Smoke:NiagaraFloat=&FloatFromCurve(FloatCurve=DataInterface() {propertyValues {Curve {keys [
          {interpMode=RCIM_Cubic, tangentMode=RCTM_User, time=0, value=1}
          {arriveTangent=-1, interpMode=RCIM_Cubic, leaveTangent=-1, time=0.3521, value=0.3455}
          <at 1 0>
         ]}}})
     01_DirtClods:NiagaraFloat=&FloatFromCurve(FloatCurve=DataInterface() {propertyValues {Curve {keys [
          {interpMode=RCIM_Cubic, tangentMode=RCTM_User, time=0, value=1}
          {arriveTangent=-1, interpMode=RCIM_Cubic, leaveTangent=-1, time=0.5, value=0.35}
          <at 1 0>
         ]}}})
     02_Grass:NiagaraFloat=1
     03_Sand:NiagaraFloat=&FloatFromCurve(
      FloatCurve=DataInterface() {propertyValues {Curve {keys [
          {arriveTangent=4.1785, interpMode=RCIM_Cubic, leaveTangent=4.1785, tangentMode=RCTM_User, time=0, value=0}
          {arriveTangent=0.9995, interpMode=RCIM_Cubic, leaveTangent=0.9995, tangentMode=RCTM_User, time=0.0178, value=1}
          {arriveTangent=-1.0182, interpMode=RCIM_Cubic, leaveTangent=-1.0182, time=0.5124, value=0.3091}
          {arriveTangent=-0.3824, interpMode=RCIM_Cubic, leaveTangent=-0.0405, tangentMode=RCTM_Break, time=1, value=0}
         ]}}}
      Scale Curve=&FNTimeOfDayScalar(Night Scale=0.5, Morning Scale=0.5, Day Scale=0.7, Evening Scale=0.5)
     )
     04_SnowMist:NiagaraFloat=&Multiply_Float(
      A=&FloatFromCurve(
       FloatCurve=DataInterface() {propertyValues {Curve {keys [{interpMode=RCIM_Cubic, leaveTangent=0.1086, tangentMode=RCTM_Break, time=0, value=0.95}, {tangentMode=RCTM_User, time=1, value=0}]}}}
       Scale Curve=&RandomRangeFloat(Minimum=0.5, Maximum=1)
      )
      B=&FNTimeOfDayScalar(Night Scale=1, Morning Scale=0.2, Evening Scale=0.2)
     )
     05_SnowChunks:NiagaraFloat=&FloatFromCurve(
      FloatCurve=DataInterface() {propertyValues {Curve {keys [
          {arriveTangent=4.1785, interpMode=RCIM_Cubic, leaveTangent=4.1785, tangentMode=RCTM_User, time=0, value=0}
          {arriveTangent=0.9995, interpMode=RCIM_Cubic, leaveTangent=0.9995, tangentMode=RCTM_User, time=0.0178, value=1}
          {arriveTangent=-1.0182, interpMode=RCIM_Cubic, leaveTangent=-1.0182, time=0.5124, value=0.3091}
          {arriveTangent=-0.3824, interpMode=RCIM_Cubic, leaveTangent=-0.0405, tangentMode=RCTM_Break, time=1, value=0}
         ]}}}
      Scale Curve=0.7
     )
     07_Ice Mesh:NiagaraFloat=1
     08_Ice Shiny Flakes:NiagaraFloat=1
     09_Mud Kickup:NiagaraFloat=&FloatFromCurve(
      FloatCurve=DataInterface() {propertyValues {Curve {keys [
          {arriveTangent=4.1785, interpMode=RCIM_Cubic, leaveTangent=4.1785, tangentMode=RCTM_User, time=0, value=0}
          {arriveTangent=0.9995, interpMode=RCIM_Cubic, leaveTangent=0.9995, tangentMode=RCTM_User, time=0.0178, value=1}
          {arriveTangent=-1.0182, interpMode=RCIM_Cubic, leaveTangent=-1.0182, time=0.6, value=0.55}
          {arriveTangent=-0.3824, interpMode=RCIM_Cubic, leaveTangent=-0.0405, tangentMode=RCTM_Break, time=1, value=0}
         ]}}}
      Scale Curve=0.7
     )
     10_Mud Splash:NiagaraFloat=&FloatFromCurve(
      FloatCurve=DataInterface() {propertyValues {Curve {keys [
          {arriveTangent=4.1785, interpMode=RCIM_Cubic, leaveTangent=4.1785, tangentMode=RCTM_User, time=0, value=0}
          {arriveTangent=0.9995, interpMode=RCIM_Cubic, leaveTangent=0.9995, tangentMode=RCTM_User, time=0.0178, value=1}
          {arriveTangent=-1.0182, interpMode=RCIM_Cubic, leaveTangent=-1.0182, time=0.6, value=0.55}
          {arriveTangent=-0.3824, interpMode=RCIM_Cubic, leaveTangent=-0.0405, tangentMode=RCTM_Break, time=1, value=0}
         ]}}}
      Scale Curve=0.7
     )
     11_Mud Peel Out Splash:NiagaraFloat=&FloatFromCurve(FloatCurve=DataInterface() {propertyValues {Curve {keys [<at 0 0>, {arriveTangent=8.6954, interpMode=RCIM_Cubic, leaveTangent=8.6954, tangentMode=RCTM_User, time=0.0788, value=1.0113}, <at 1 0>]}}})
     12_Peel Out Spikes:NiagaraFloat=&FloatFromCurve(FloatCurve=DataInterface() {propertyValues {Curve {keys [<at 0 0>, {arriveTangent=8.6954, interpMode=RCIM_Cubic, leaveTangent=8.6954, tangentMode=RCTM_User, time=0.0788, value=1.0113}, <at 1 0>]}}})
    )
    Scale RGB=&self:SpawnGroupVectorSelection(
     00_Smoke:Vector3f=&VectorFromFloat(Value=&FloatFromCurve(FloatCurve=DataInterface() {propertyValues {Curve {keys [{arriveTangent=1.8145, interpMode=RCIM_Cubic, leaveTangent=0.8876, tangentMode=RCTM_Break, time=0, value=0.75}, <at 0.3387 0.9565>, <at 1 1>]}}}))
     01_DirtClods:Vector3f=&VectorFromFloat(Value=&FloatFromCurve(FloatCurve=DataInterface() {propertyValues {Curve {keys [{arriveTangent=1.8145, interpMode=RCIM_Cubic, leaveTangent=0.8876, tangentMode=RCTM_Break, time=0, value=0.75}, <at 0.3387 0.9565>, <at 1 1>]}}}))
     02_Grass:Vector3f=<1 1 1>
     03_Sand:Vector3f=<1 1 1>
     04_SnowMist:Vector3f=<1 1 1>
     05_SnowChunks:Vector3f=<1 1 1>
     07_IceMesh:Vector3f=<1 1 1>
     08_IceShinyFlakes:Vector3f=<1 1 1>
     09_MudKickup:Vector3f=<1 1 1>
     10_MudSplash:Vector3f=<1 1 1>
     11_Mud PeelOut Splash:Vector3f=&VectorFromFloat(Value=1.5)
     12_Peel Out Spikes:Vector3f=&VectorFromFloat(Value=1.5)
    )
    ScaleA=true
    ScaleRGB=true
   )
   &SpriteRotationRate(Rotation Rate=&self:SpawnGroupFloatSelection(
     00_Smoke:NiagaraFloat=$Particles.Initial.MaterialRandom
     01_DirtClods:NiagaraFloat=$Particles.Initial.MaterialRandom
     02_Grass:NiagaraFloat=0
     03_Sand:NiagaraFloat=$Particles.Initial.MaterialRandom
     04_SnowMist:NiagaraFloat=$Particles.Initial.MaterialRandom
     05_SnowChunks:NiagaraFloat=$Particles.Initial.MaterialRandom
     07_Ice Mesh:NiagaraFloat=0
     08_Ice Shiny Flakes:NiagaraFloat=200
     09_Mud Kickup:NiagaraFloat=0
     10_Mud Splash:NiagaraFloat=0
     11_Mud Peel Out Splash:NiagaraFloat=0
     12_Peel Out Spikes:NiagaraFloat=0
    ))
   &GravityForce(Gravity=&self:SpawnGroupVectorSelection(
     00_Smoke:Vector3f=<0 0 -2940>
     01_DirtClods:Vector3f=<0 0 -2940>
     02_Grass:Vector3f=<0 0 -784>
     03_Sand:Vector3f=<0 0 -2940>
     04_SnowMist:Vector3f=<0 0 -2940>
     05_SnowChunks:Vector3f=<0 0 -4900>
     07_IceMesh:Vector3f=<0 0 -2940>
     08_IceShinyFlakes:Vector3f=<0 0 -49>
     09_MudKickup:Vector3f=<0 0 -9800>
     10_MudSplash:Vector3f=<0 0 -9800>
     11_Mud PeelOut Splash:Vector3f=<0 0 0>
     12_Peel Out Spikes:Vector3f=<0 0 0>
    ))
   &Drag(
    Drag=&self:SpawnGroupFloatSelection(
     00_Smoke:NiagaraFloat=1
     01_DirtClods:NiagaraFloat=2
     02_Grass:NiagaraFloat=3
     03_Sand:NiagaraFloat=1
     04_SnowMist:NiagaraFloat=1
     05_SnowChunks:NiagaraFloat=1
     07_Ice Mesh:NiagaraFloat=2
     08_Ice Shiny Flakes:NiagaraFloat=&RandomRangeFloat(Minimum=1, Maximum=2)
     09_Mud Kickup:NiagaraFloat=1
     10_Mud Splash:NiagaraFloat=7
     11_Mud Peel Out Splash:NiagaraFloat=0
     12_Peel Out Spikes:NiagaraFloat=0
    )
    Rotational Drag:NiagaraFloat=&self:SpawnGroupFloatSelection(
     00_Smoke:NiagaraFloat=1
     01_DirtClods:NiagaraFloat=1
     02_Grass:NiagaraFloat=0
     03_Sand:NiagaraFloat=1
     04_SnowMist:NiagaraFloat=1
     05_SnowChunks:NiagaraFloat=1
     07_Ice Mesh:NiagaraFloat=1
     08_Ice Shiny Flakes:NiagaraFloat=1
     09_Mud Kickup:NiagaraFloat=1
     10_Mud Splash:NiagaraFloat=1
     11_Mud Peel Out Splash:NiagaraFloat=0
     12_Peel Out Spikes:NiagaraFloat=0
    )
    Use Linear Drag=true
    Use Rotational Drag=true
   )
   &InheritSourceMovement(
    Applied Position Scale=&self:SpawnGroupVectorSelection(
     00_Smoke:Vector3f=<0.9 0.9 0.9>
     01_DirtClods:Vector3f=<0.9 0.9 0.9>
     02_Grass:Vector3f=<0.9 0.9 0.9>
     03_Sand:Vector3f=<0.9 0.9 0.9>
     04_SnowMist:Vector3f=<0.9 0.9 0.9>
     05_SnowChunks:Vector3f=<0 0 0>
     07_IceMesh:Vector3f=<0 0 0>
     08_IceShinyFlakes:Vector3f=<0 0 0>
     09_MudKickup:Vector3f=<0.1 0.1 0.1>
     10_MudSplash:Vector3f=<0 0 0>
     11_Mud PeelOut Splash:Vector3f=<0 0 0>
     12_Peel Out Spikes:Vector3f=<0 0 0>
    )
    ApplyForce=false
    ApplyPosition=true
    ApplyPositionFalloff=false
   )
   &SolveForcesAndVelocity(Clamp Velocity=false, Limit Acceleration=false)
   &DynamicMaterialParameters(
    Index 0 Param 1=&self:SpawnGroupFloatSelection(
     00_Smoke:NiagaraFloat=&Add_Float(
      A=&RandomRangeFloat(Minimum=0, Maximum=1)
      B=&FloatFromCurve(FloatCurve=DataInterface() {propertyValues {Curve {keys [
           {arriveTangent=2.7533, interpMode=RCIM_Cubic, leaveTangent=2.847, tangentMode=RCTM_Break, time=0, value=0}
           {arriveTangent=1.3581, interpMode=RCIM_Cubic, leaveTangent=1.3581, tangentMode=RCTM_User, time=0.2558, value=0.5125}
           {arriveTangent=0.0529, interpMode=RCIM_Cubic, leaveTangent=0.2021, tangentMode=RCTM_User, time=1, value=1}
          ]}}})
     )
     01_DirtClods:NiagaraFloat=0
     02_Grass:NiagaraFloat=0
     03_Sand:NiagaraFloat=&FloatFromCurve(FloatCurve=DataInterface() {propertyValues {Curve {keys [<at 0 -1>,<at 1 1>]}}})
     04_SnowMist:NiagaraFloat=1
     05_SnowChunks:NiagaraFloat=0
     07_Ice Mesh:NiagaraFloat=0
     08_Ice Shiny Flakes:NiagaraFloat=0
     09_Mud Kickup:NiagaraFloat=0
     10_Mud Splash:NiagaraFloat=0
     11_Mud Peel Out Splash:NiagaraFloat=&FloatFromCurve(FloatCurve=DataInterface() {propertyValues {Curve {keys [<at -0.0046 1.0113>,<at 1.0011 -1>]}}})
     12_Peel Out Spikes:NiagaraFloat=&FloatFromCurve(FloatCurve=DataInterface() {propertyValues {Curve {keys [<at -0.0046 1.0113>,<at 1.0011 -1>]}}})
    )
    Index 0 Param 2=&self:SpawnGroupFloatSelection(
     00_Smoke:NiagaraFloat=0
     01_DirtClods:NiagaraFloat=0
     02_Grass:NiagaraFloat=0
     03_Sand:NiagaraFloat=&FNTimeOfDayScalar(Night Scale=0.7, Morning Scale=0.7, Day Scale=0.1, Evening Scale=0.5)
     04_SnowMist:NiagaraFloat=1
     05_SnowChunks:NiagaraFloat=0
     07_Ice Mesh:NiagaraFloat=0
     08_Ice Shiny Flakes:NiagaraFloat=0
     09_Mud Kickup:NiagaraFloat=0
     10_Mud Splash:NiagaraFloat=0
     11_Mud Peel Out Splash:NiagaraFloat=0
     12_Peel Out Spikes:NiagaraFloat=0
    )
    Index 0 Param 3=&self:SpawnGroupFloatSelection(
     00_Smoke:NiagaraFloat=0
     01_DirtClods:NiagaraFloat=0
     02_Grass:NiagaraFloat=0
     03_Sand:NiagaraFloat=&FNTimeOfDayScalar(Night Scale=2, Morning Scale=2, Evening Scale=1.5)
     04_SnowMist:NiagaraFloat=1
     05_SnowChunks:NiagaraFloat=0
     07_Ice Mesh:NiagaraFloat=0
     08_Ice Shiny Flakes:NiagaraFloat=0
     09_Mud Kickup:NiagaraFloat=0
     10_Mud Splash:NiagaraFloat=0
     11_Mud Peel Out Splash:NiagaraFloat=0
     12_Peel Out Spikes:NiagaraFloat=0
    )
    Index 0 Param 4=&self:SpawnGroupFloatSelection(
     00_Smoke:NiagaraFloat=&FloatFromCurve(FloatCurve=DataInterface() {propertyValues {Curve {keys [<at 0 1>,<at 1 0>]}}})
     01_DirtClods:NiagaraFloat=0
     02_Grass:NiagaraFloat=0
     03_Sand:NiagaraFloat=&FNTimeOfDayScalar(Night Scale=0.02, Morning Scale=-0.01, Day Scale=0, Evening Scale=-0.01)
     04_SnowMist:NiagaraFloat=1
     05_SnowChunks:NiagaraFloat=0
     07_Ice Mesh:NiagaraFloat=0
     08_Ice Shiny Flakes:NiagaraFloat=0
     09_Mud Kickup:NiagaraFloat=0
     10_Mud Splash:NiagaraFloat=0
     11_Mud Peel Out Splash:NiagaraFloat=0.5
     12_Peel Out Spikes:NiagaraFloat=0.5
    )
    Index 2 Param 1=&self:SpawnGroupFloatSelection(
     00_Smoke:NiagaraFloat=&FloatFromCurve(FloatCurve=DataInterface() {propertyValues {Curve {keys [
          {arriveTangent=2.646, interpMode=RCIM_Cubic, leaveTangent=2.646, tangentMode=RCTM_User, time=0, value=0}
          {leaveTangent=-0.0719, time=1, value=1}
         ]}}})
     01_DirtClods:NiagaraFloat=0
     02_Grass:NiagaraFloat=0
     03_Sand:NiagaraFloat=0
     04_SnowMist:NiagaraFloat=0
     05_SnowChunks:NiagaraFloat=0
     07_Ice Mesh:NiagaraFloat=0
     08_Ice Shiny Flakes:NiagaraFloat=0
     09_Mud Kickup:NiagaraFloat=0
     10_Mud Splash:NiagaraFloat=0
     11_Mud Peel Out Splash:NiagaraFloat=0
     12_Peel Out Spikes:NiagaraFloat=0
    )
    Index 2 Param 2=&self:SpawnGroupFloatSelection(
     00_Smoke:NiagaraFloat=&FloatFromCurve(
      FloatCurve=DataInterface() {propertyValues {Curve {keys [
          {arriveTangent=1.9056, time=0, value=0.2}
          {arriveTangent=0.0259, interpMode=RCIM_Cubic, leaveTangent=0.0259, tangentMode=RCTM_User, time=1, value=1.4}
         ]}}}
      Scale Curve=&RandomRangeFloat(Minimum=0.5, Maximum=1)
     )
     01_DirtClods:NiagaraFloat=0
     02_Grass:NiagaraFloat=0
     03_Sand:NiagaraFloat=0
     04_SnowMist:NiagaraFloat=0
     05_SnowChunks:NiagaraFloat=0
     07_Ice Mesh:NiagaraFloat=0
     08_Ice Shiny Flakes:NiagaraFloat=0
     09_Mud Kickup:NiagaraFloat=0
     10_Mud Splash:NiagaraFloat=0
     11_Mud Peel Out Splash:NiagaraFloat=0
     12_Peel Out Spikes:NiagaraFloat=0
    )
    Index 2 Param 3=&self:SpawnGroupFloatSelection(
     00_Smoke:NiagaraFloat=&FloatFromCurve(FloatCurve=DataInterface() {propertyValues {Curve {keys [
          {interpMode=RCIM_Cubic, leaveTangent=2.3064, tangentMode=RCTM_Break, time=0, value=1}
          {interpMode=RCIM_Cubic, leaveTangent=11.1233, tangentMode=RCTM_Break, time=1, value=2}
         ]}}})
     01_DirtClods:NiagaraFloat=0
     02_Grass:NiagaraFloat=0
     03_Sand:NiagaraFloat=0
     04_SnowMist:NiagaraFloat=0
     05_SnowChunks:NiagaraFloat=0
     07_Ice Mesh:NiagaraFloat=0
     08_Ice Shiny Flakes:NiagaraFloat=0
     09_Mud Kickup:NiagaraFloat=0
     10_Mud Splash:NiagaraFloat=0
     11_Mud Peel Out Splash:NiagaraFloat=0
     12_Peel Out Spikes:NiagaraFloat=0
    )
    Index 2 Param 4=&self:SpawnGroupFloatSelection(
     00_Smoke:NiagaraFloat=&FloatFromCurve(
      FloatCurve=DataInterface() {propertyValues {Curve {keys [
          {arriveTangent=-0.6867, interpMode=RCIM_Cubic, leaveTangent=-0.6867, tangentMode=RCTM_User, time=0, value=1}
          {interpMode=RCIM_Cubic, leaveTangent=-1.2956, tangentMode=RCTM_Break, time=1, value=0.75}
         ]}}}
      Scale Curve=1.6
     )
     01_DirtClods:NiagaraFloat=0
     02_Grass:NiagaraFloat=0
     03_Sand:NiagaraFloat=0
     04_SnowMist:NiagaraFloat=0
     05_SnowChunks:NiagaraFloat=0
     07_Ice Mesh:NiagaraFloat=0
     08_Ice Shiny Flakes:NiagaraFloat=0
     09_Mud Kickup:NiagaraFloat=0
     10_Mud Splash:NiagaraFloat=0
     11_Mud Peel Out Splash:NiagaraFloat=0
     12_Peel Out Spikes:NiagaraFloat=0
    )
    Module.Index0Param0Write=true
    Module.Index0Param1Write=true
    Module.Index0Param2Write=true
    Module.Index0Param3Write=true
    Module.Index2Param0Write=true
    Module.Index2Param1Write=true
    Module.Index2Param2Write=true
    Module.Index2Param3Write=true
    Write Parameter Index 2=true
   )
   &ScaleSpriteSize(
    Scale Factor=&self:SpawnGroupVector2DSelection(
     00_Smoke:Vector2f=&Vector2DFromFloat(Value=&FloatFromCurve(
       FloatCurve=DataInterface() {propertyValues {Curve {keys [
           {interpMode=RCIM_Cubic, leaveTangent=1.2011, tangentMode=RCTM_Break, time=0, value=0.179}
           {arriveTangent=0.6691, interpMode=RCIM_Cubic, leaveTangent=0.6691, tangentMode=RCTM_User, time=0.8329, value=1}
          ]}}}
       Scale Curve=&Multiply_Float(
        A=&Lerp_Float(A=3.5, B=5.5, Alpha=$System.Timeline002.Time)
        B=&MakeCustomFloatFromBool(Bool=$User.IsPeelOutOnRoad, True Float=0.6, False Float=1)
       )
      ))
     01_DirtClods:Vector2f=&Vector2DFromFloat(Value=&FloatFromCurve(
       FloatCurve=DataInterface() {propertyValues {Curve {keys [
           {interpMode=RCIM_Cubic, leaveTangent=0.9665, tangentMode=RCTM_Break, time=0, value=0.1}
           {arriveTangent=1.0806, interpMode=RCIM_Cubic, leaveTangent=1.0806, time=0.3541, value=0.5745}
           {arriveTangent=1.9934, interpMode=RCIM_Cubic, leaveTangent=1.9934, tangentMode=RCTM_User, time=0.8329, value=1}
          ]}}}
       Scale Curve=6
      ))
     02_Grass:Vector2f=<0 0>
     03_Sand:Vector2f=&Vector2DFromFloat(Value=&FloatFromCurve(
       FloatCurve=DataInterface() {propertyValues {Curve {keys [
           {interpMode=RCIM_Cubic, leaveTangent=1.2011, tangentMode=RCTM_Break, time=0, value=0.179}
           {arriveTangent=0.6691, interpMode=RCIM_Cubic, leaveTangent=0.6691, tangentMode=RCTM_User, time=0.8329, value=1}
          ]}}}
       Scale Curve=&Multiply_Float(
        A=&Lerp_Float(A=3.5, B=5.5, Alpha=$System.Timeline002.Time)
        B=&MakeCustomFloatFromBool(Bool=$User.IsPeelOutOnRoad, True Float=0.6, False Float=1)
       )
      ))
     04_SnowMist:Vector2f=&Multiply_Vector2D(
      A=&Vector2DFromFloat(Value=&FloatFromCurve(
        FloatCurve=DataInterface() {propertyValues {Curve {keys [
            {arriveTangent=4.1785, interpMode=RCIM_Cubic, leaveTangent=4.1785, tangentMode=RCTM_User, time=0, value=0.5}
            {arriveTangent=1.1179, interpMode=RCIM_Cubic, leaveTangent=1.1179, tangentMode=RCTM_User, time=0.2042, value=0.8}
            <at 0.9 1>
            {arriveTangent=0.0224, interpMode=RCIM_Cubic, leaveTangent=0.0224, tangentMode=RCTM_User, time=1, value=0}
           ]}}}
        Scale Curve=&Multiply_Float(
         A=&Lerp_Float(A=1.3, B=1.6, Alpha=$System.Timeline002.Time)
         B=&MakeCustomFloatFromBool(Bool=$User.IsPeelOutOnRoad, True Float=0.6, False Float=1)
        )
       ))
      B=&Vector2DFromFloat(Value=&FloatFromCurve(
        FloatCurve=DataInterface() {propertyValues {Curve {keys [
            {interpMode=RCIM_Cubic, leaveTangent=1.2011, tangentMode=RCTM_Break, time=0, value=0.179}
            {arriveTangent=0.6691, interpMode=RCIM_Cubic, leaveTangent=0.6691, tangentMode=RCTM_User, time=0.8329, value=1}
           ]}}}
        Scale Curve=&Multiply_Float(
         A=&Lerp_Float(A=0.8, B=1, Alpha=$System.Timeline002.Time)
         B=&MakeCustomFloatFromBool(Bool=$User.IsPeelOutOnRoad, True Float=0.6, False Float=1)
        )
       ))
     )
     05_SnowChunks:Vector2f=&Vector2DFromFloat(Value=&FloatFromCurve(FloatCurve=DataInterface() {propertyValues {Curve {keys [{tangentMode=RCTM_User, time=0, value=1}, {tangentMode=RCTM_User, time=1, value=0}]}}}))
     07_IceMesh:Vector2f=<0 0>
     08_IceShinyFlakes:Vector2f=&Vector2DFromFloat(Value=&FloatFromCurve(FloatCurve=DataInterface() {propertyValues {Curve {keys [
           {arriveTangent=3.9985, interpMode=RCIM_Cubic, leaveTangent=3.9985, tangentMode=RCTM_User, time=0, value=0}
           <at 0.5 1 RCIM_Cubic>
           {arriveTangent=-5.042, interpMode=RCIM_Cubic, leaveTangent=-5.042, tangentMode=RCTM_User, time=1, value=0}
          ]}}}))
     09_Mud Kickup:Vector2f=&Vector2DFromFloat(Value=&FloatFromCurve(FloatCurve=DataInterface() {propertyValues {Curve {keys [{tangentMode=RCTM_User, time=0, value=1}, {tangentMode=RCTM_User, time=1, value=0}]}}}))
     10_Mud Splash:Vector2f=&Vector2DFromFloat(Value=&FloatFromCurve(FloatCurve=DataInterface() {propertyValues {Curve {keys [{tangentMode=RCTM_User, time=0, value=1}, {tangentMode=RCTM_User, time=1, value=0}]}}}))
     11_Mud PeelOut Splash:Vector2f=&Vector2DFromFloat(Value=&FloatFromCurve(
       FloatCurve=DataInterface() {propertyValues {Curve {keys [{arriveTangent=2.3953, interpMode=RCIM_Cubic, leaveTangent=2.3953, tangentMode=RCTM_User, time=0, value=0.2}, <at 1 1>]}}}
       Scale Curve=3.5
      ))
     12_PeelOut Spikes:Vector2f=&Vector2DFromFloat(Value=&FloatFromCurve(FloatCurve=DataInterface() {propertyValues {Curve {keys [{arriveTangent=2.3953, interpMode=RCIM_Cubic, leaveTangent=2.3953, tangentMode=RCTM_User, time=0, value=0.2}, <at 1 1>]}}}))
    )
    Scale Sprite Size Mode=::NewEnumerator2
   )
   &ScaleMeshSize(Scale Factor=&self:SpawnGroupVectorSelection(
     00_Smoke:Vector3f=<0 0 0>
     01_DirtClods:Vector3f=<0 0 0>
     02_Grass:Vector3f=&VectorFromFloat(Value=&FloatFromCurve(FloatCurve=DataInterface() {propertyValues {Curve {keys [
           {arriveTangent=5, interpMode=RCIM_Cubic, leaveTangent=5, time=0, value=0}
           {arriveTangent=1.0417, interpMode=RCIM_Cubic, leaveTangent=1.0417, time=0.2, value=1}
           {arriveTangent=-1.25, interpMode=RCIM_Cubic, leaveTangent=-1.25, time=0.6, value=0.625}
           {arriveTangent=0.095, interpMode=RCIM_Cubic, leaveTangent=0.095, tangentMode=RCTM_User, time=1, value=0}
          ]}}}))
     03_Sand:Vector3f=<0 0 0>
     04_SnowMist:Vector3f=<0 0 0>
     05_SnowChunks:Vector3f=<0 0 0>
     07_IceMesh:Vector3f=&VectorFromFloat(Value=&FloatFromCurve(
       FloatCurve=DataInterface() {propertyValues {Curve {keys [
           {interpMode=RCIM_Cubic, tangentMode=RCTM_User, time=0, value=0}
           <at 0.15 0.95>
           {arriveTangent=-1.0223, interpMode=RCIM_Cubic, leaveTangent=-1.0223, tangentMode=RCTM_User, time=1, value=0}
          ]}}}
       Scale Curve=0.15
      ))
     08_IceShinyFlakes:Vector3f=<0 0 0>
     09_MudKickup:Vector3f=<0 0 0>
     10_MudSplash:Vector3f=<0 0 0>
     11_Mud PeelOut Splash:Vector3f=<0 0 0>
     12_Peel Out Spikes:Vector3f=<0 0 0>
    ))
   &UpdateMeshOrientation(
    Rotation Rate=&RandomRangeFloat(Minimum=-0.4, Maximum=0.4)
    Rotation Vector=&RandomVector(Vector Scale=1)
   )
   /SetVariables(Particles.Snow_SubUVIndex:NiagaraFloat=&FloatFromCurve(FloatCurve=DataInterface() {propertyValues {Curve {keys [{tangentMode=RCTM_User, time=0, value=0}, {tangentMode=RCTM_User, time=1, value=63}]}}})) {isSetParameters=true}
  ]
  rendererClasses [&NiagaraSpriteRendererProperties, &NiagaraMeshRendererProperties]
  renderers [
   {rendererClass=&NiagaraSpriteRendererProperties}
   {rendererClass=&NiagaraSpriteRendererProperties, rendererIndex=1}
   {rendererClass=&NiagaraSpriteRendererProperties, rendererIndex=2}
   {rendererClass=&NiagaraMeshRendererProperties, rendererIndex=3}
   {rendererClass=&NiagaraSpriteRendererProperties, rendererIndex=4}
   {rendererClass=&NiagaraSpriteRendererProperties, rendererIndex=5}
   {rendererClass=&NiagaraSpriteRendererProperties, rendererIndex=6}
   {rendererClass=&NiagaraMeshRendererProperties, rendererIndex=7}
   {rendererClass=&NiagaraMeshRendererProperties, rendererIndex=8}
   {rendererClass=&NiagaraSpriteRendererProperties, rendererIndex=9}
   {rendererClass=&NiagaraSpriteRendererProperties, rendererIndex=10}
   {rendererClass=&NiagaraSpriteRendererProperties, rendererIndex=11}
   {rendererClass=&NiagaraSpriteRendererProperties, rendererIndex=12}
   {rendererClass=&NiagaraSpriteRendererProperties, rendererIndex=13}
  ]
 }
 SkidMarks {
  emitterUpdateScript [
   &EmitterState(Life Cycle Mode=::NewEnumerator1, Loop Duration=1, UseLoopDelay=false)
   &SpawnPerUnit(
    Module.Use Max Movement Threshold=false
    Module.Use Movement Tolerance=false
    Spawn Per Unit=70
    Spawn Probability:NiagaraFloat=&MakeCustomFloatFromBool(
     Bool=&Boolean_AndOperation(
      Bool1=&Boolean_AndOperation(Bool1=$User.PeelOut_Wheel_Rear, Bool2=&Boolean_NotOperation(Bool=$User.IsDoingStoppie))
      Bool2=&Boolean_NotOperation(Bool=&Boolean_OrOperation(
        Bool1=$System.SpawnSand
        Bool2=&Boolean_OrOperation(Bool1=$System.SpawnSnow, Bool2=&Boolean_OrOperation(Bool1=$System.SpawnIce, Bool2=$System.SpawnMud))
       ))
     )
     True Float=1
     False Float=0
    )
    Use Spawn Probability=true
   )
  ]
  particleSpawnScript [
   &InitializeParticle(Color=<rgba 0 0 0 0.6>, Lifetime=5, UsePositionOffset=false, Write Material Random=true) {module=InitializeParticle001}
   /SetVariables() {isSetParameters=true, module=SetVariables_#5}
   /SetVariables(Particles.RibbonWidth=20) {isSetParameters=true, module=SetVariables_#6}
   &SkeletalMeshLocation(
    Bone / Socket Index=1
    Inherit Velocity=false
    Interpolate Mesh Sampling=false
    MaskSpawnGroup=false
    Mesh Sampling Type=::NewEnumerator5
    Sampled Position Offset Coordinate Space:ENiagaraMeshTransforms=::NewEnumerator3
    Skeletal Mesh=$System.SkeletalMesh
    Socket Sampling Mode=::NewEnumerator5
   )
  ]
  particleUpdateScript [
   &ParticleState()
   &SolveForcesAndVelocity(Clamp Velocity=false, Limit Acceleration=false)
   &CameraOffset(Camera Offset Amount=32)
   &ScaleColor(
    Scale Alpha=&FloatFromCurve(
     FloatCurve=DataInterface() {propertyValues {Curve {keys [{interpMode=RCIM_Cubic, tangentMode=RCTM_Break, time=0, value=1}, <at 1 0>]}}}
     Scale Curve=&FloatFromCurve(
      FloatCurve=DataInterface() {propertyValues {Curve {keys [{interpMode=RCIM_Cubic, tangentMode=RCTM_Break, time=0, value=0}, {interpMode=RCIM_Cubic, tangentMode=RCTM_Break, time=0.1153, value=1}]}}}
      CurveIndex=$Particles.Age
     )
    )
    ScaleA=true
    ScaleRGB=true
   )
   &KillParticles(Kill Particles=$User.IsInWater)
  ]
  rendererClasses [&NiagaraRibbonRendererProperties]
  renderers [
   {rendererClass=&NiagaraRibbonRendererProperties}
   {rendererClass=&NiagaraRibbonRendererProperties, rendererIndex=1}
  ]
 }
 Sparks_Tires_Blown {
  emitterUpdateScript [
   &EmitterState()
   &SpawnRate(
    SpawnRate=&Multiply_Float(
     A=&Lerp_Float(
      A=&RandomRangeFloat(Minimum=20, Maximum=30)
      B=&RandomRangeFloat(Minimum=40, Maximum=50)
      Alpha=&Power(a=$User.Vehicle_Speed, b=1.25)
     )
     B=&Multiply_Float(
      A=&MakeFloatFromInt(Int=$System.Sparks_FrontTire)
      B=&FloatFromCurve(FloatCurve=$System.Sparks_SpawnCurve, CurveIndex=&Abs_Float(Float:NiagaraFloat=$User.Vehicle_Speed))
     )
    )
    Use Spawn Probability=false
   )
   &SpawnRate(
    SpawnGroup=1
    SpawnRate=&Multiply_Float(
     A=&Lerp_Float(
      A=&RandomRangeFloat(Minimum=20, Maximum=30)
      B=&RandomRangeFloat(Minimum=40, Maximum=50)
      Alpha=&Power(a=$User.Vehicle_Speed, b=1.25)
     )
     B=&Multiply_Float(
      A=&MakeFloatFromInt(Int=$System.Sparks_RearTire)
      B=&FloatFromCurve(FloatCurve=$System.Sparks_SpawnCurve, CurveIndex=&Abs_Float(Float:NiagaraFloat=$User.Vehicle_Speed))
     )
    )
    Use Spawn Probability=false
   ) {module=SpawnRate001}
   &SpawnBurst_Instantaneous(Spawn Count=$System.Sparks_FrontTire, Spawn Group=2, Use Spawn Probability=false, UseLoopCountLimit=false)
   &SpawnBurst_Instantaneous(Spawn Count=$System.Sparks_RearTire, Spawn Group=3, Use Spawn Probability=false, UseLoopCountLimit=false) {module=SpawnBurst_Instantaneous001}
  ]
  particleSpawnScript [
   /SetVariables(Particles.VisibilityTag=&SelectIntFromArray(Direct Array Index=&SpawnGroup(), Int Selection Array=DataInterface() {propertyValues {IntData [0,0,1,1]}})) {isSetParameters=true}
   &InitializeParticle(
    Color Maximum=&MakeCustomLinearColorFromBool(
     Bool=&SetBoolByIntComparison(A=&SpawnGroup(), B=2, Comparison Type=::NewEnumerator2)
     True Color=<rgba 150 41.4953 0 1>
     False Color=<rgba 40 0.5 0 1>
    )
    Color Minimum=&MakeCustomLinearColorFromBool(
     Bool=&SetBoolByIntComparison(A=&SpawnGroup(), B=2, Comparison Type=::NewEnumerator2)
     True Color=<rgba 50 9.9924 0 1>
     False Color=<rgba 25 15.7812 0 1>
    )
    Color Mode=::NewEnumerator2
    Lifetime Max=&MakeCustomFloatFromBool(
     Bool=&SetBoolByIntComparison(A=&SpawnGroup(), B=2, Comparison Type=::NewEnumerator2)
     True Float=0.1
     False Float=0.4
    )
    Lifetime Min=0.05
    Lifetime Mode=::NewEnumerator1
    Mesh Scale=&MaskVectorBySpawnGroup(Masked Vector=<0.205 0.15 0.205>, Passthrough Vector=<0.26 0.11 0.26>, Spawn Group Mask=3)
    Mesh Scale Mode=::NewEnumerator4
    Sprite Size Max=<2.5 30>
    Sprite Size Min=<1.25 20>
    Sprite Size Mode=::NewEnumerator2
    UsePositionOffset=false
    Write Material Random=true
   )
   &SkeletalMeshLocation(
    Bone / Socket Index=&SpawnGroup()
    MaskSpawnGroup=false
    Mesh Sampling Type=::NewEnumerator5
    Sampled Position Offset=&SelectVectorFromArray(
     Direct Array Index=&SpawnGroup()
     Vector Selection Array=DataInterface() {propertyValues {FloatData [<0 0 0>, <0 0 6>, <0 0 0>, <0 0 0>]}}
    )
    Skeletal Mesh=$System.SkeletalMesh
    Socket Sampling Mode=::NewEnumerator5
   )
   /SetVariables(Particles.ParticleVelocityAdd:Vector3f=&RotateVectorByQuaternion(
     Vector To Rotate=&MakeVector(
      X=&RandomRangeFloat(Minimum=-100, Maximum=-400)
      Y=0
      Z=&Multiply_Float(
       A=&RandomRangeFloat(Minimum=10, Maximum=200)
       B=&Abs_Float(Float:NiagaraFloat=$User.Vehicle_Speed)
      )
     )
     Quaternion=$StackContext.SkeletalMeshLocation.SampledRotation
    )) {isSetParameters=true}
   /SetVariables(Particles.Velocity=&MakeCustomVectorFromBool(
     Bool=&SetBoolByIntComparison(A=&SpawnGroup(), B=2, Comparison Type=::NewEnumerator6)
     True Vector=&Add_Vector(A=$Particles.Velocity, B=$Particles.ParticleVelocityAdd)
     False Vector=<0 0 0>
    )) {isSetParameters=true}
   /SetVariables(Particles.SpriteAlignment=&NormalizeVector(Vector To Normalize=$Particles.ParticleVelocityAdd, Safe Normalize=true, Safe Fallback Threshold=0, Safe Fallback Vector=<0 0 0.001>)) {isSetParameters=true}
   &InitialMeshOrientation(
    Mesh Orientation Mode=::NewEnumerator2
    Orientation Vector=<0 1 0>
    Rotation=&RandomRangeVector(Minimum=<0 0 0>, Maximum=<0 1 0>)
    Use Rotation Vector=true
   )
  ]
  particleUpdateScript [
   &ParticleState()
   &SolveForcesAndVelocity(Clamp Velocity=false, Limit Acceleration=false)
   &ScaleColor(
    Scale Alpha=&FloatFromCurve(
     FloatCurve=DataInterface() {propertyValues {Curve {keys [<at -0.0046 1.0113 RCIM_Cubic>,<at 1.0011 0.002 RCIM_Cubic>]}}}
     Scale Curve=&MakeCustomFloatFromBool(
      Bool=&SetBoolByIntComparison(A=&SpawnGroup(), B=2, Comparison Type=::NewEnumerator2)
      True Float=&ClampFloat(Float=&Multiply_Float(A=2, B=&Abs_Float(Float:NiagaraFloat=&Power(a=$User.Vehicle_Speed, b=2.5))))
      False Float=1
     )
    )
    Scale RGB=&MakeCustomVectorFromBool(
     Bool=&SetBoolByIntComparison(A=&SpawnGroup(), B=2, Comparison Type=::NewEnumerator2)
     True Vector=<100 100 100>
     False Vector=<1 1 1>
    )
    ScaleA=true
    ScaleRGB=true
   )
   &ScaleSpriteSize(
    Uniform Curve Scale=&Lerp_Float(A=0.2, B=1, Alpha=$User.Vehicle_Speed)
    Uniform Curve Sprite Scale=DataInterface() {propertyValues {Curve {keys [
        {arriveTangent=56.0592, interpMode=RCIM_Cubic, leaveTangent=56.0592, tangentMode=RCTM_User, time=0, value=0}
        {arriveTangent=-1.0335, interpMode=RCIM_Cubic, leaveTangent=-1.0335, tangentMode=RCTM_User, time=0.1674, value=2.5}
        <at 1 0>
       ]}}}
   )
   &ScaleSpriteSizeBySpeed(Max Scale Factor=<1 1>, Min Scale Factor=<0.5 0.6>)
   &DynamicMaterialParameters(
    Index 0 Param 1=5
    Index 0 Param 2=2
    Module.Index0Param0Write=true
    Module.Index0Param1Write=true
    Module.Index0Param2Write=false
    Module.Index0Param3Write=false
   )
   &SkeletalMeshLocation(
    Bone / Socket Index=2
    MaskSpawnGroup=true
    Mesh Sampling Type=::NewEnumerator5
    Skeletal Mesh=$System.SkeletalMesh
    Socket Sampling Mode=::NewEnumerator5
    Spawn Group Mask=2
   ) {module=SkeletalMeshLocation001}
   &SkeletalMeshLocation(
    Bone / Socket Index=3
    MaskSpawnGroup=true
    Mesh Sampling Type=::NewEnumerator5
    Skeletal Mesh=$System.SkeletalMesh
    Socket Sampling Mode=::NewEnumerator5
    Spawn Group Mask=3
   ) {module=SkeletalMeshLocation002}
   /SetVariables(Particles.MeshOrientation=&MultiplyQuaternion(
     Quaternion A=/Niagara/DynamicInputs/TypeConversions/MakeCustomQuaternionFromBool(
      Bool:NiagaraBool=&SetBoolByIntComparison(A=&SpawnGroup(), B=3, Comparison Type=::NewEnumerator6)
      True Quat:Quat4f=$StackContext.SkeletalMeshLocation001.SampledRotation
      False Quat:Quat4f=$StackContext.SkeletalMeshLocation002.SampledRotation
     )
     Quaternion B=&OrientMeshToVector(Look At Direction=&RandomRangeVector(Minimum=<0 0 0>, Maximum=<0 1 0>), Reference Axis Vector=<1 0 0>)
    )) {isSetParameters=true}
  ]
  rendererClasses [&NiagaraSpriteRendererProperties, &NiagaraMeshRendererProperties]
  renderers [
   {rendererClass=&NiagaraSpriteRendererProperties}
   {rendererClass=&NiagaraMeshRendererProperties, rendererIndex=1}
  ]
 }
}
summary {
 systemName=NS_Motorcycle_Parent_Wheel_And_Tire_Interaction
 emitters [
  {emitterName=SkidMarks, rendererClasses [&NiagaraRibbonRendererProperties]}
  {emitterName=NE_Parent_MudTrail, rendererClasses [&NiagaraRibbonRendererProperties]}
  {
   emitterName=Sparks_Tires_Blown
   rendererClasses [&NiagaraSpriteRendererProperties, &NiagaraMeshRendererProperties]
  }
  {emitterName=Dust_Debris, rendererClasses [&NiagaraSpriteRendererProperties]}
  {emitterName=PeelOut, rendererClasses [&NiagaraSpriteRendererProperties, &NiagaraMeshRendererProperties]}
  {
   emitterName=Exhaust_and_ShiftBlip
   rendererClasses [&NiagaraSpriteRendererProperties, &NiagaraLightRendererProperties]
  }
 ]
 userVariables [
  User.ChargeProgress {type=NiagaraFloat, defaultValue=0, description=""}
  User.Dust_Wheel_Rear {type=NiagaraBool}
  User.IsBoostJump {type=NiagaraBool}
  User.IsDoingStoppie {type=NiagaraBool}
  User.IsInAir {type=NiagaraBool}
  User.IsInWater {type=NiagaraBool}
  User.IsOnWheelie {type=NiagaraBool}
  User.IsPeelOutOnRoad {type=NiagaraBool}
  User.IsPeelOutOnRoadRewind {type=NiagaraBool}
  User.IsShifting {type=NiagaraBool}
  User.IsWaterDeep {type=NiagaraBool}
  User.PeelOut_Wheel_Rear {type=NiagaraBool}
  User.PeelOutColorIndex {type=NiagaraInt32}
  User.Slide_Direction {type=NiagaraFloat, defaultValue=0, description=""}
  User.SpawnExhaustFlames {type=NiagaraBool}
  User.TirePoppedSparksFront {type=NiagaraBool}
  User.TirePoppedSparksRear {type=NiagaraBool}
  User.Vehicle_Speed {type=NiagaraFloat, defaultValue=0, description=""}
  User.VehicleHasDriver {type=NiagaraBool}
  User.VehicleHasFuel {type=NiagaraBool}
 ]
}
